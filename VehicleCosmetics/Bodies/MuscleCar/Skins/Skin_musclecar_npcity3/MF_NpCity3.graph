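# Material function for the muscle car's NpCity3 body skin, building a full material attribute set.
# An RGB skin mask drives team, accent and paint colors plus finish, roughness, clear coat, detail normals and emissive through composite subgraphs.
# Stacked, switchable texture layers add time-driven panning, distortion and sine pulsing, each blended over the last.
# Legacy Rocket League sharp-mask subgraphs process the skin mask and multiply base color.
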MaterialFunction /VehicleCosmetics/Bodies/MuscleCar/Skins/Skin_musclecar_npcity3/MF_NpCity3
graph {expressions {
  MaterialExpressionAbs_0 {class=MaterialExpressionAbs, inputs [None {from=MaterialExpressionAdd_0, output=""}], properties {}}
  MaterialExpressionAdd_0 {
   class=MaterialExpressionAdd
   inputs [
    A {from=MaterialExpressionConstant_0}
    B {from=MaterialExpressionPower_1}
   ]
  }
  MaterialExpressionAdd_1 {
   class=MaterialExpressionAdd
   inputs [
    A {from=MaterialExpressionComponentMask_1}
    B {from=MaterialExpressionComponentMask_1}
   ]
  }
  MaterialExpressionAdd_10 {
   class=MaterialExpressionAdd
   inputs [
    A {from=MaterialExpressionStaticSwitchParameter_33}
    B {from=MaterialExpressionStaticSwitchParameter_4}
   ]
  }
  MaterialExpressionAdd_11 {
   class=MaterialExpressionAdd
   inputs [
    A {from=MaterialExpressionMultiply_38}
    B {from=MaterialExpressionMultiply_39}
   ]
  }
  MaterialExpressionAdd_12 {
   class=MaterialExpressionAdd
   inputs [
    A {from=MaterialExpressionStaticSwitchParameter_44}
    B {from=MaterialExpressionStaticSwitchParameter_43}
   ]
  }
  MaterialExpressionAdd_13 {
   class=MaterialExpressionAdd
   inputs [
    A {from=MaterialExpressionMultiply_47}
    B {from=MaterialExpressionMultiply_48}
   ]
  }
  MaterialExpressionAdd_14 {
   class=MaterialExpressionAdd
   inputs [
    A {from=MaterialExpressionStaticSwitchParameter_56}
    B {from=MaterialExpressionStaticSwitchParameter_55}
   ]
  }
  MaterialExpressionAdd_2 {
   class=MaterialExpressionAdd
   inputs [
    A {from=MaterialExpressionStaticSwitchParameter_34}
    B {from=MaterialExpressionMultiply_2}
   ]
  }
  MaterialExpressionAdd_3 {
   class=MaterialExpressionAdd
   inputs [
    A {from=MaterialExpressionMultiply_21}
    B {from=MaterialExpressionAppendVector_12}
   ]
  }
  MaterialExpressionAdd_4 {
   class=MaterialExpressionAdd
   inputs [
    A {from=MaterialExpressionAdd_1}
    B {from=MaterialExpressionPower_5}
   ]
  }
  MaterialExpressionAdd_5 {
   class=MaterialExpressionAdd
   inputs [
    A {from=MaterialExpressionMultiply_11}
    B {from=MaterialExpressionMultiply_12}
   ]
  }
  MaterialExpressionAdd_6 {
   class=MaterialExpressionAdd
   inputs [
    A {from=MaterialExpressionStaticSwitchParameter_10}
    B {from=MaterialExpressionStaticSwitchParameter_12}
   ]
  }
  MaterialExpressionAdd_7 {
   class=MaterialExpressionAdd
   inputs [
    A {from=MaterialExpressionMultiply_19}
    B {from=MaterialExpressionMultiply_20}
   ]
  }
  MaterialExpressionAdd_8 {
   class=MaterialExpressionAdd
   inputs [
    A {from=MaterialExpressionStaticSwitchParameter_22}
    B {from=MaterialExpressionStaticSwitchParameter_21}
   ]
  }
  MaterialExpressionAppendVector_0 {
   class=MaterialExpressionAppendVector
   inputs [
    A {from=MaterialExpressionTextureSampleParameter2D_0, output=B}
    B {from=MaterialExpressionTextureSampleParameter2D_0, output=B}
   ]
  }
  MaterialExpressionAppendVector_1 {
   class=MaterialExpressionAppendVector
   inputs [
    A {from=MaterialExpressionComponentMask_4, output=""}
    B {from=MaterialExpressionComponentMask_5, output=""}
   ]
  }
  MaterialExpressionAppendVector_10 {
   class=MaterialExpressionAppendVector
   inputs [
    A {from=MaterialExpressionStaticSwitchParameter_20, output=""}
    B {from=MaterialExpressionStaticComponentMaskParameter_1, output=""}
   ]
  }
  MaterialExpressionAppendVector_11 {
   class=MaterialExpressionAppendVector
   inputs [
    A {from=MaterialExpressionVectorParameter_12, output=R}
    B {from=MaterialExpressionVectorParameter_12, output=R}
   ]
  }
  MaterialExpressionAppendVector_12 {
   class=MaterialExpressionAppendVector
   inputs [
    A {from=MaterialExpressionVectorParameter_12, output=B}
    B {from=MaterialExpressionVectorParameter_12, output=B}
   ]
  }
  MaterialExpressionAppendVector_15 {
   class=MaterialExpressionAppendVector
   inputs [
    A {from=MaterialExpressionVectorParameter_14, output=R}
    B {from=MaterialExpressionVectorParameter_14, output=R}
   ]
  }
  MaterialExpressionAppendVector_16 {
   class=MaterialExpressionAppendVector
   inputs [
    A {from=MaterialExpressionVectorParameter_14, output=B}
    B {from=MaterialExpressionVectorParameter_14, output=B}
   ]
  }
  MaterialExpressionAppendVector_17 {
   class=MaterialExpressionAppendVector
   inputs [
    A {from=MaterialExpressionTextureSampleParameter2D_12, output=RGB}
    B {from=MaterialExpressionTextureSampleParameter2D_12, output=RGB}
   ]
  }
  MaterialExpressionAppendVector_18 {
   class=MaterialExpressionAppendVector
   inputs [
    A {from=MaterialExpressionStaticSwitchParameter_42, output=""}
    B {from=MaterialExpressionStaticComponentMaskParameter_3, output=""}
   ]
  }
  MaterialExpressionAppendVector_19 {
   class=MaterialExpressionAppendVector
   inputs [
    A {from=MaterialExpressionVectorParameter_16, output=R}
    B {from=MaterialExpressionVectorParameter_16, output=R}
   ]
  }
  MaterialExpressionAppendVector_2 {
   class=MaterialExpressionAppendVector
   inputs [
    A {from=MaterialExpressionVectorParameter_8, output=R}
    B {from=MaterialExpressionVectorParameter_8, output=R}
   ]
  }
  MaterialExpressionAppendVector_20 {
   class=MaterialExpressionAppendVector
   inputs [
    A {from=MaterialExpressionVectorParameter_16, output=B}
    B {from=MaterialExpressionVectorParameter_16, output=B}
   ]
  }
  MaterialExpressionAppendVector_21 {
   class=MaterialExpressionAppendVector
   inputs [
    A {from=MaterialExpressionTextureSampleParameter2D_13, output=RGB}
    B {from=MaterialExpressionTextureSampleParameter2D_13, output=RGB}
   ]
  }
  MaterialExpressionAppendVector_22 {
   class=MaterialExpressionAppendVector
   inputs [
    A {from=MaterialExpressionStaticSwitchParameter_54, output=""}
    B {from=MaterialExpressionStaticComponentMaskParameter_4, output=""}
   ]
  }
  MaterialExpressionAppendVector_3 {
   class=MaterialExpressionAppendVector
   inputs [
    A {from=MaterialExpressionVectorParameter_8, output=B}
    B {from=MaterialExpressionVectorParameter_8, output=B}
   ]
  }
  MaterialExpressionAppendVector_4 {
   class=MaterialExpressionAppendVector
   inputs [
    A {from=MaterialExpressionTextureSampleParameter2D_8, output=RGB}
    B {from=MaterialExpressionTextureSampleParameter2D_8, output=RGB}
   ]
  }
  MaterialExpressionAppendVector_5 {
   class=MaterialExpressionAppendVector
   inputs [
    A {from=MaterialExpressionStaticSwitchParameter_11, output=""}
    B {from=MaterialExpressionStaticComponentMaskParameter_0, output=""}
   ]
  }
  MaterialExpressionAppendVector_6 {
   class=MaterialExpressionAppendVector
   inputs [
    A {from=MaterialExpressionVectorParameter_10, output=R}
    B {from=MaterialExpressionVectorParameter_10, output=R}
   ]
  }
  MaterialExpressionAppendVector_7 {
   class=MaterialExpressionAppendVector
   inputs [
    A {from=MaterialExpressionVectorParameter_10, output=B}
    B {from=MaterialExpressionVectorParameter_10, output=B}
   ]
  }
  MaterialExpressionAppendVector_8 {
   class=MaterialExpressionAppendVector
   inputs [
    A {from=MaterialExpressionOneMinus_2, output=""}
    B {from=MaterialExpressionComponentMask_7, output=""}
   ]
  }
  MaterialExpressionAppendVector_9 {
   class=MaterialExpressionAppendVector
   inputs [
    A {from=MaterialExpressionTextureSampleParameter2D_10, output=RGB}
    B {from=MaterialExpressionTextureSampleParameter2D_10, output=RGB}
   ]
  }
  MaterialExpressionChannelMaskParameter_1 {
   class=MaterialExpressionChannelMaskParameter
   inputs [None {from=MaterialExpressionAppendVector_1, output=""}]
   properties {
    desc="R: SkinMask - Red; G: SkinMask - Green; B: SkinMask - Blue; A: BodyTextures - Blue"
    group=Color
    maskChannel=Alpha
    parameterName=EmissiveMaskSelector
    sortPriority=6
    defaultValue=<rgba 0 0 0 1>
   }
  }
  MaterialExpressionClamp_0 {
   class=MaterialExpressionClamp
   inputs [
    None {from=MaterialExpressionStaticSwitchParameter_15, output=""}
    Min {from="", output=""}
    Max {from="", output=""}
   ]
   properties {}
  }
  MaterialExpressionClamp_1 {
   class=MaterialExpressionClamp
   inputs [
    None {from=MaterialExpressionPower_7, output=""}
    Min {from=MaterialExpressionConstant_6, output=""}
    Max {from="", output=""}
   ]
   properties {clampMode=CMODE_ClampMin, minDefault=0.5}
  }
  MaterialExpressionClamp_10 {
   class=MaterialExpressionClamp
   inputs [
    None {from=MaterialExpressionSubtract_5, output=""}
    Min {from="", output=""}
    Max {from="", output=""}
   ]
   properties {}
  }
  MaterialExpressionClamp_2 {
   class=MaterialExpressionClamp
   inputs [
    None {from=MaterialExpressionSubtract_0, output=""}
    Min {from="", output=""}
    Max {from="", output=""}
   ]
   properties {}
  }
  MaterialExpressionClamp_3 {
   class=MaterialExpressionClamp
   inputs [
    None {from=MaterialExpressionStaticSwitchParameter_27, output=""}
    Min {from="", output=""}
    Max {from="", output=""}
   ]
   properties {}
  }
  MaterialExpressionClamp_4 {
   class=MaterialExpressionClamp
   inputs [
    None {from=MaterialExpressionSubtract_2, output=""}
    Min {from="", output=""}
    Max {from="", output=""}
   ]
   properties {}
  }
  MaterialExpressionClamp_7 {
   class=MaterialExpressionClamp
   inputs [
    None {from=MaterialExpressionStaticSwitchParameter_49, output=""}
    Min {from="", output=""}
    Max {from="", output=""}
   ]
   properties {}
  }
  MaterialExpressionClamp_8 {
   class=MaterialExpressionClamp
   inputs [
    None {from=MaterialExpressionSubtract_4, output=""}
    Min {from="", output=""}
    Max {from="", output=""}
   ]
   properties {}
  }
  MaterialExpressionClamp_9 {
   class=MaterialExpressionClamp
   inputs [
    None {from=MaterialExpressionStaticSwitchParameter_61, output=""}
    Min {from="", output=""}
    Max {from="", output=""}
   ]
   properties {}
  }
  MaterialExpressionComponentMask_0 {class=MaterialExpressionComponentMask, inputs [None {from=MaterialExpressionReroute_30, output=SkinMaskTexture}], properties {r=true}}
  MaterialExpressionComponentMask_1 {class=MaterialExpressionComponentMask, inputs [None {from=MaterialExpressionAbs_0}], properties {g=true}}
  MaterialExpressionComponentMask_2 {class=MaterialExpressionComponentMask, inputs [None {from=MaterialExpressionNamedRerouteUsage_34}], properties {r=true}}
  MaterialExpressionComponentMask_3 {class=MaterialExpressionComponentMask, inputs [None {from=MaterialExpressionNamedRerouteUsage_35}], properties {g=true}}
  MaterialExpressionComponentMask_4 {
   class=MaterialExpressionComponentMask
   inputs [None {from=MaterialExpressionReroute_55, output=SkinMaskTexture}]
   properties {b=true, g=true, r=true}
  }
  MaterialExpressionComponentMask_5 {class=MaterialExpressionComponentMask, inputs [None {from=MaterialExpressionSaturate_4}], properties {r=true}}
  MaterialExpressionComponentMask_6 {class=MaterialExpressionComponentMask, inputs [None {from=MaterialExpressionTextureCoordinate_10}], properties {g=true}}
  MaterialExpressionComponentMask_7 {class=MaterialExpressionComponentMask, inputs [None {from=MaterialExpressionTextureCoordinate_10}], properties {r=true}}
  MaterialExpressionComposite_0 {
   class=MaterialExpressionComposite
   inputs [SkinMaskTexture {from=MaterialExpressionNamedRerouteUsage_12, output=""}]
   properties {subgraphName=BodyFinish}
  }
  MaterialExpressionComposite_1 {
   class=MaterialExpressionComposite
   inputs [
    UV {from=MaterialExpressionNamedRerouteUsage_25, output=""}
    UV Detail {from=MaterialExpressionNamedRerouteUsage_26, output=""}
    UV BottomNormal {from=MaterialExpressionNamedRerouteUsage_21, output=""}
    LOD {from=MaterialExpressionScalarParameter_1, output=""}
    BottomNormalFlatness {from=MaterialExpressionScalarParameter_14, output=""}
   ]
   properties {subgraphName=BodyTextures}
  }
  MaterialExpressionComposite_10 {
   class=MaterialExpressionComposite
   inputs [
    BaseColorInput {from=MaterialExpressionReroute_28, output=BaseColor}
    TextureInput {from=MaterialExpressionNamedRerouteUsage_27, output=""}
   ]
   properties {desc=Legacy - Rocket League Feature, subgraphName=SharpMasksBaseColorMultiply}
  }
  MaterialExpressionComposite_2 {
   class=MaterialExpressionComposite
   inputs [
    DetailBaseTexture {from=MaterialExpressionNamedRerouteUsage_29, output=""}
    DetailAccentTexture {from=MaterialExpressionNamedRerouteUsage_30, output=""}
    SkinMaskTexture {from=MaterialExpressionNamedRerouteUsage_18, output=""}
   ]
   properties {subgraphName=Details}
  }
  MaterialExpressionComposite_3 {
   class=MaterialExpressionComposite
   inputs [
    Roughness {from=MaterialExpressionNamedRerouteUsage_5, output=""}
    SkinMask {from=MaterialExpressionNamedRerouteUsage_4, output=""}
   ]
   properties {subgraphName=Roughness}
  }
  MaterialExpressionComposite_4 {
   class=MaterialExpressionComposite
   inputs [
    TeamColor {from=MaterialExpressionNamedRerouteUsage_8, output=""}
    AccentColor {from=MaterialExpressionNamedRerouteUsage_9, output=""}
    PaintColor {from=MaterialExpressionNamedRerouteUsage_10, output=""}
    SkinMaskTexture {from=MaterialExpressionNamedRerouteUsage_7, output=""}
   ]
   properties {subgraphName=BaseColor}
  }
  MaterialExpressionComposite_5 {
   class=MaterialExpressionComposite
   inputs [
    BaseColor {from=MaterialExpressionReroute_26, output=BaseColor}
    BaseColorTexture {from=MaterialExpressionNamedRerouteUsage_15, output=""}
    SkinMaskTexture {from=MaterialExpressionNamedRerouteUsage_14, output=""}
   ]
   properties {subgraphName=UseBaseColorTexture}
  }
  MaterialExpressionComposite_6 {class=MaterialExpressionComposite, inputs [], properties {subgraphName=Vectors}}
  MaterialExpressionComposite_7 {
   class=MaterialExpressionComposite
   inputs [
    SkinMaskTexture {from=MaterialExpressionNamedRerouteUsage_13, output=""}
    BottomNormalBaseTexture {from=MaterialExpressionNamedRerouteUsage_11, output=""}
    BottomNormalAccentTexture {from=MaterialExpressionNamedRerouteUsage_19, output=""}
   ]
   properties {subgraphName=ClearCoat}
  }
  MaterialExpressionComposite_8 {
   class=MaterialExpressionComposite
   inputs [RGB {from=MaterialExpressionReroute_44, output=SkinMask}]
   properties {desc=Legacy - Rocket League Feature, subgraphName=ProcessSharpMasks}
  }
  MaterialExpressionComposite_9 {
   class=MaterialExpressionComposite
   inputs [
    SkinMaskTexture {from=MaterialExpressionNamedRerouteUsage_39, output=""}
    BodyTextures {from=MaterialExpressionNamedRerouteUsage_38, output=""}
    EmissiveColor {from=MaterialExpressionNamedRerouteUsage_33, output=""}
   ]
   properties {subgraphName=EmissiveColor}
  }
  MaterialExpressionConstant3Vector_1 {class=MaterialExpressionConstant3Vector, properties {constant=<rgba 0 0 1 0>}}
  MaterialExpressionConstantBiasScale_0 {
   class=MaterialExpressionConstantBiasScale
   inputs [None {from=MaterialExpressionStaticComponentMaskParameter_0}]
   properties {bias=-0.5, scale=2}
  }
  MaterialExpressionConstantBiasScale_1 {class=MaterialExpressionConstantBiasScale, inputs [None {from=MaterialExpressionPower_0}], properties {bias=-0.5, scale=4}}
  MaterialExpressionConstantBiasScale_10 {class=MaterialExpressionConstantBiasScale, inputs [None {from=MaterialExpressionSine_4}], properties {bias=1.25}}
  MaterialExpressionConstantBiasScale_2 {class=MaterialExpressionConstantBiasScale, inputs [None {from=MaterialExpressionSine_0}], properties {bias=1.25}}
  MaterialExpressionConstantBiasScale_3 {
   class=MaterialExpressionConstantBiasScale
   inputs [None {from=MaterialExpressionStaticComponentMaskParameter_1}]
   properties {bias=-0.5, scale=2}
  }
  MaterialExpressionConstantBiasScale_4 {class=MaterialExpressionConstantBiasScale, inputs [None {from=MaterialExpressionSine_1}], properties {bias=1.25}}
  MaterialExpressionConstantBiasScale_7 {
   class=MaterialExpressionConstantBiasScale
   inputs [None {from=MaterialExpressionStaticComponentMaskParameter_3}]
   properties {bias=-0.5, scale=2}
  }
  MaterialExpressionConstantBiasScale_8 {class=MaterialExpressionConstantBiasScale, inputs [None {from=MaterialExpressionSine_3}], properties {bias=1.25}}
  MaterialExpressionConstantBiasScale_9 {
   class=MaterialExpressionConstantBiasScale
   inputs [None {from=MaterialExpressionStaticComponentMaskParameter_4}]
   properties {bias=-0.5, scale=2}
  }
  MaterialExpressionConstant_0 {class=MaterialExpressionConstant, properties {r=-0.5}}
  MaterialExpressionConstant_1 {class=MaterialExpressionConstant}
  MaterialExpressionConstant_10 {class=MaterialExpressionConstant, properties {r=0.1}}
  MaterialExpressionConstant_11 {class=MaterialExpressionConstant}
  MaterialExpressionConstant_12 {class=MaterialExpressionConstant}
  MaterialExpressionConstant_13 {class=MaterialExpressionConstant}
  MaterialExpressionConstant_14 {class=MaterialExpressionConstant, properties {r=1}}
  MaterialExpressionConstant_2 {class=MaterialExpressionConstant, properties {r=0.1}}
  MaterialExpressionConstant_21 {class=MaterialExpressionConstant, properties {r=0.1}}
  MaterialExpressionConstant_22 {class=MaterialExpressionConstant}
  MaterialExpressionConstant_23 {class=MaterialExpressionConstant}
  MaterialExpressionConstant_24 {class=MaterialExpressionConstant}
  MaterialExpressionConstant_25 {class=MaterialExpressionConstant, properties {r=1}}
  MaterialExpressionConstant_26 {class=MaterialExpressionConstant}
  MaterialExpressionConstant_27 {class=MaterialExpressionConstant, properties {r=0.1}}
  MaterialExpressionConstant_28 {class=MaterialExpressionConstant}
  MaterialExpressionConstant_29 {class=MaterialExpressionConstant}
  MaterialExpressionConstant_3 {class=MaterialExpressionConstant}
  MaterialExpressionConstant_30 {class=MaterialExpressionConstant}
  MaterialExpressionConstant_31 {class=MaterialExpressionConstant, properties {r=1}}
  MaterialExpressionConstant_32 {class=MaterialExpressionConstant}
  MaterialExpressionConstant_4 {class=MaterialExpressionConstant}
  MaterialExpressionConstant_5 {class=MaterialExpressionConstant, properties {r=1}}
  MaterialExpressionConstant_6 {class=MaterialExpressionConstant, properties {r=0.5}}
  MaterialExpressionConstant_7 {class=MaterialExpressionConstant}
  MaterialExpressionConstant_8 {class=MaterialExpressionConstant}
  MaterialExpressionFunctionOutput_0 {class=MaterialExpressionFunctionOutput, inputs [None {from=MaterialExpressionSetMaterialAttributes_0, output=""}]}
  MaterialExpressionLinearInterpolate_0 {
   class=MaterialExpressionLinearInterpolate
   inputs [
    A {from=MaterialExpressionScalarParameter_2}
    B {from=MaterialExpressionScalarParameter_3}
    Alpha {from=MaterialExpressionMaterialFunctionCall_0, output=G}
   ]
  }
  MaterialExpressionLinearInterpolate_1 {
   class=MaterialExpressionLinearInterpolate
   inputs [
    A {from=MaterialExpressionLinearInterpolate_0}
    B {from=MaterialExpressionScalarParameter_4}
    Alpha {from=MaterialExpressionMaterialFunctionCall_0, output=B}
   ]
  }
  MaterialExpressionLinearInterpolate_10 {
   class=MaterialExpressionLinearInterpolate
   inputs [
    A {from=MaterialExpressionReroute_29, output=BaseColorTexture}
    B {from=MaterialExpressionReroute_27, output=BaseColor}
    Alpha {from=MaterialExpressionComponentMask_0}
   ]
   properties {desc=Put diffuse where R skin mask is black}
  }
  MaterialExpressionLinearInterpolate_11 {
   class=MaterialExpressionLinearInterpolate
   inputs [
    A {from=MaterialExpressionScalarParameter_8}
    B {from=MaterialExpressionScalarParameter_9}
    Alpha {from=MaterialExpressionMaterialFunctionCall_8, output=G}
   ]
  }
  MaterialExpressionLinearInterpolate_12 {
   class=MaterialExpressionLinearInterpolate
   inputs [
    A {from=MaterialExpressionLinearInterpolate_11}
    B {from=MaterialExpressionScalarParameter_10}
    Alpha {from=MaterialExpressionMaterialFunctionCall_8, output=B}
   ]
  }
  MaterialExpressionLinearInterpolate_13 {
   class=MaterialExpressionLinearInterpolate
   inputs [
    A {from=MaterialExpressionReroute_35, output=BottomNormalBaseTexture}
    B {from=MaterialExpressionReroute_36, output=BottomNormalAccentTexture}
    Alpha {from=MaterialExpressionMaterialFunctionCall_3, output=G}
   ]
  }
  MaterialExpressionLinearInterpolate_14 {
   class=MaterialExpressionLinearInterpolate
   inputs [
    A {from=MaterialExpressionLinearInterpolate_13}
    B {from=MaterialExpressionReroute_36, output=BottomNormalAccentTexture}
    Alpha {from=MaterialExpressionMaterialFunctionCall_3, output=B}
   ]
  }
  MaterialExpressionLinearInterpolate_15 {
   class=MaterialExpressionLinearInterpolate
   inputs [
    A {from=MaterialExpressionReroute_31, output=DetailBaseTexture}
    B {from=MaterialExpressionReroute_34, output=DetailAccentTexture}
    Alpha {from=MaterialExpressionSaturate_1}
   ]
   properties {constA=1}
  }
  MaterialExpressionLinearInterpolate_16 {
   class=MaterialExpressionLinearInterpolate
   inputs [
    A {from=MaterialExpressionStaticSwitchParameter_8}
    B {from=MaterialExpressionStaticSwitchParameter_14}
    Alpha {from=MaterialExpressionMultiply_15}
   ]
  }
  MaterialExpressionLinearInterpolate_17 {
   class=MaterialExpressionLinearInterpolate
   inputs [
    A {from=MaterialExpressionScalarParameter_11}
    B {from=MaterialExpressionScalarParameter_12}
    Alpha {from=MaterialExpressionMaterialFunctionCall_19, output=G}
   ]
  }
  MaterialExpressionLinearInterpolate_18 {
   class=MaterialExpressionLinearInterpolate
   inputs [
    A {from=MaterialExpressionLinearInterpolate_17}
    B {from=MaterialExpressionScalarParameter_13}
    Alpha {from=MaterialExpressionMaterialFunctionCall_19, output=B}
   ]
  }
  MaterialExpressionLinearInterpolate_19 {
   class=MaterialExpressionLinearInterpolate
   inputs [
    A {from=MaterialExpressionStaticSwitchParameter_16}
    B {from=MaterialExpressionConstant_4}
    Alpha {from=MaterialExpressionTextureSampleParameter2D_9, output=A}
   ]
  }
  MaterialExpressionLinearInterpolate_2 {
   class=MaterialExpressionLinearInterpolate
   inputs [
    A {from=MaterialExpressionScalarParameter_5}
    B {from=MaterialExpressionScalarParameter_6}
    Alpha {from=MaterialExpressionMaterialFunctionCall_2, output=G}
   ]
  }
  MaterialExpressionLinearInterpolate_20 {
   class=MaterialExpressionLinearInterpolate
   inputs [
    A {from=MaterialExpressionConstant_5}
    B {from=MaterialExpressionConstantBiasScale_2}
    Alpha {from=MaterialExpressionScalarParameter_22}
   ]
  }
  MaterialExpressionLinearInterpolate_21 {
   class=MaterialExpressionLinearInterpolate
   inputs [
    A {from=MaterialExpressionMultiply_9}
    B {from=MaterialExpressionMultiply_18}
    Alpha {from=MaterialExpressionClamp_0}
   ]
  }
  MaterialExpressionLinearInterpolate_22 {
   class=MaterialExpressionLinearInterpolate
   inputs [
    A {from=MaterialExpressionConstant_8}
    B {from=MaterialExpressionStaticSwitchParameter_19}
    Alpha {from=MaterialExpressionScalarParameter_25}
   ]
  }
  MaterialExpressionLinearInterpolate_23 {
   class=MaterialExpressionLinearInterpolate
   inputs [
    A {from=MaterialExpressionConstant_12}
    B {from=MaterialExpressionStaticSwitchParameter_24}
    Alpha {from=MaterialExpressionScalarParameter_29}
   ]
  }
  MaterialExpressionLinearInterpolate_24 {
   class=MaterialExpressionLinearInterpolate
   inputs [
    A {from=MaterialExpressionLinearInterpolate_16}
    B {from=MaterialExpressionStaticSwitchParameter_26}
    Alpha {from=MaterialExpressionMultiply_25}
   ]
  }
  MaterialExpressionLinearInterpolate_25 {
   class=MaterialExpressionLinearInterpolate
   inputs [
    A {from=MaterialExpressionStaticSwitchParameter_28}
    B {from=MaterialExpressionConstant_13}
    Alpha {from=MaterialExpressionTextureSampleParameter2D_9, output=A}
   ]
  }
  MaterialExpressionLinearInterpolate_26 {
   class=MaterialExpressionLinearInterpolate
   inputs [
    A {from=MaterialExpressionConstant_14}
    B {from=MaterialExpressionConstantBiasScale_4}
    Alpha {from=MaterialExpressionScalarParameter_32}
   ]
  }
  MaterialExpressionLinearInterpolate_27 {
   class=MaterialExpressionLinearInterpolate
   inputs [
    A {from=MaterialExpressionStaticSwitchParameter_18}
    B {from=MaterialExpressionMultiply_28}
    Alpha {from=MaterialExpressionClamp_3}
   ]
  }
  MaterialExpressionLinearInterpolate_28 {
   class=MaterialExpressionLinearInterpolate
   inputs [
    A {from=MaterialExpressionLinearInterpolate_39}
    B {from=MaterialExpressionTextureSampleParameter2D_11, output=RGB}
    Alpha {from=MaterialExpressionTextureSampleParameter2D_11, output=RGB}
   ]
  }
  MaterialExpressionLinearInterpolate_29 {
   class=MaterialExpressionLinearInterpolate
   inputs [
    A {from=MaterialExpressionStaticSwitchParameter_64}
    B {from=MaterialExpressionMultiply_29}
    Alpha {from=MaterialExpressionTextureSampleParameter2D_11, output=A}
   ]
  }
  MaterialExpressionLinearInterpolate_3 {
   class=MaterialExpressionLinearInterpolate
   inputs [
    A {from=MaterialExpressionLinearInterpolate_2}
    B {from=MaterialExpressionScalarParameter_7}
    Alpha {from=MaterialExpressionMaterialFunctionCall_2, output=B}
   ]
  }
  MaterialExpressionLinearInterpolate_33 {
   class=MaterialExpressionLinearInterpolate
   inputs [
    A {from=MaterialExpressionConstant_23}
    B {from=MaterialExpressionStaticSwitchParameter_46}
    Alpha {from=MaterialExpressionScalarParameter_43}
   ]
  }
  MaterialExpressionLinearInterpolate_34 {
   class=MaterialExpressionLinearInterpolate
   inputs [
    A {from=MaterialExpressionLinearInterpolate_24}
    B {from=MaterialExpressionStaticSwitchParameter_48}
    Alpha {from=MaterialExpressionMultiply_43}
   ]
  }
  MaterialExpressionLinearInterpolate_35 {
   class=MaterialExpressionLinearInterpolate
   inputs [
    A {from=MaterialExpressionStaticSwitchParameter_53}
    B {from=MaterialExpressionConstant_24}
    Alpha {from=MaterialExpressionTextureSampleParameter2D_9, output=A}
   ]
  }
  MaterialExpressionLinearInterpolate_36 {
   class=MaterialExpressionLinearInterpolate
   inputs [
    A {from=MaterialExpressionConstant_25}
    B {from=MaterialExpressionConstantBiasScale_8}
    Alpha {from=MaterialExpressionScalarParameter_46}
   ]
  }
  MaterialExpressionLinearInterpolate_37 {
   class=MaterialExpressionLinearInterpolate
   inputs [
    A {from=MaterialExpressionStaticSwitchParameter_30}
    B {from=MaterialExpressionMultiply_46}
    Alpha {from=MaterialExpressionClamp_7}
   ]
  }
  MaterialExpressionLinearInterpolate_38 {
   class=MaterialExpressionLinearInterpolate
   inputs [
    A {from=MaterialExpressionConstant_29}
    B {from=MaterialExpressionStaticSwitchParameter_58}
    Alpha {from=MaterialExpressionScalarParameter_51}
   ]
  }
  MaterialExpressionLinearInterpolate_39 {
   class=MaterialExpressionLinearInterpolate
   inputs [
    A {from=MaterialExpressionLinearInterpolate_34}
    B {from=MaterialExpressionStaticSwitchParameter_60}
    Alpha {from=MaterialExpressionMultiply_52}
   ]
  }
  MaterialExpressionLinearInterpolate_4 {
   class=MaterialExpressionLinearInterpolate
   inputs [
    A {from=MaterialExpressionConstant3Vector_1}
    B {from=MaterialExpressionTextureSampleParameter2D_5, output=RGB}
    Alpha {from=MaterialExpressionReroute_21, output=LOD}
   ]
   properties {constAlpha=0}
  }
  MaterialExpressionLinearInterpolate_40 {
   class=MaterialExpressionLinearInterpolate
   inputs [
    A {from=MaterialExpressionStaticSwitchParameter_62}
    B {from=MaterialExpressionConstant_30}
    Alpha {from=MaterialExpressionTextureSampleParameter2D_9, output=A}
   ]
  }
  MaterialExpressionLinearInterpolate_41 {
   class=MaterialExpressionLinearInterpolate
   inputs [
    A {from=MaterialExpressionConstant_31}
    B {from=MaterialExpressionConstantBiasScale_10}
    Alpha {from=MaterialExpressionScalarParameter_54}
   ]
  }
  MaterialExpressionLinearInterpolate_42 {
   class=MaterialExpressionLinearInterpolate
   inputs [
    A {from=MaterialExpressionStaticSwitchParameter_52}
    B {from=MaterialExpressionMultiply_55}
    Alpha {from=MaterialExpressionClamp_9}
   ]
  }
  MaterialExpressionLinearInterpolate_5 {
   class=MaterialExpressionLinearInterpolate
   inputs [
    A {from=MaterialExpressionMaterialFunctionCall_5, output=R}
    B {from=MaterialExpressionMaterialFunctionCall_5, output=R}
    Alpha {from=MaterialExpressionMaterialFunctionCall_4, output=G}
   ]
  }
  MaterialExpressionLinearInterpolate_6 {
   class=MaterialExpressionLinearInterpolate
   inputs [
    A {from=MaterialExpressionLinearInterpolate_5}
    B {from=MaterialExpressionMaterialFunctionCall_5, output=B}
    Alpha {from=MaterialExpressionMaterialFunctionCall_4, output=B}
   ]
  }
  MaterialExpressionLinearInterpolate_7 {
   class=MaterialExpressionLinearInterpolate
   inputs [
    A {from=MaterialExpressionReroute_8, output=TeamColor}
    B {from=MaterialExpressionReroute_13, output=AccentColor}
    Alpha {from=MaterialExpressionMaterialFunctionCall_6, output=G}
   ]
  }
  MaterialExpressionLinearInterpolate_8 {
   class=MaterialExpressionLinearInterpolate
   inputs [
    A {from=MaterialExpressionLinearInterpolate_7}
    B {from=MaterialExpressionReroute_18, output=PaintColor}
    Alpha {from=MaterialExpressionMaterialFunctionCall_6, output=B}
   ]
  }
  MaterialExpressionLinearInterpolate_9 {
   class=MaterialExpressionLinearInterpolate
   inputs [
    A {from=MaterialExpressionVectorParameter_4, output=RGB}
    B {from=MaterialExpressionTextureSampleParameter2D_9, output=RGB}
    Alpha {from=MaterialExpressionTextureSampleParameter2D_9, output=RGB}
   ]
  }
  MaterialExpressionMaterialFunctionCall_0 {
   class=MaterialExpressionMaterialFunctionCall
   inputs [Color {from=MaterialExpressionReroute_10, output=SkinMaskTexture}]
   properties {materialFunction=&SplitComponents}
  }
  MaterialExpressionMaterialFunctionCall_1 {
   class=MaterialExpressionMaterialFunctionCall
   inputs [
    Normal {from=MaterialExpressionTextureSampleParameter2D_6, output=RGB}
    Flatness {from=MaterialExpressionReroute_40}
   ]
   properties {materialFunction=&FlattenNormal}
  }
  MaterialExpressionMaterialFunctionCall_10 {
   class=MaterialExpressionMaterialFunctionCall
   inputs [Color {from=MaterialExpressionReroute_16, output=SkinMaskTexture}]
   properties {materialFunction=&SplitComponents}
  }
  MaterialExpressionMaterialFunctionCall_19 {
   class=MaterialExpressionMaterialFunctionCall
   inputs [Color {from=MaterialExpressionReroute_11, output=SkinMaskTexture}]
   properties {materialFunction=&SplitComponents}
  }
  MaterialExpressionMaterialFunctionCall_2 {
   class=MaterialExpressionMaterialFunctionCall
   inputs [Color {from=MaterialExpressionReroute_10, output=SkinMaskTexture}]
   properties {materialFunction=&SplitComponents}
  }
  MaterialExpressionMaterialFunctionCall_3 {
   class=MaterialExpressionMaterialFunctionCall
   inputs [Color {from=MaterialExpressionReroute_11, output=SkinMaskTexture}]
   properties {materialFunction=&SplitComponents}
  }
  MaterialExpressionMaterialFunctionCall_4 {
   class=MaterialExpressionMaterialFunctionCall
   inputs [Color {from=MaterialExpressionReroute_3, output=SkinMask}]
   properties {materialFunction=&SplitComponents}
  }
  MaterialExpressionMaterialFunctionCall_5 {
   class=MaterialExpressionMaterialFunctionCall
   inputs [Color {from=MaterialExpressionReroute_2, output=Roughness}]
   properties {materialFunction=&SplitComponents}
  }
  MaterialExpressionMaterialFunctionCall_6 {
   class=MaterialExpressionMaterialFunctionCall
   inputs [Color {from=MaterialExpressionReroute_20, output=SkinMaskTexture}]
   properties {materialFunction=&SplitComponents}
  }
  MaterialExpressionMaterialFunctionCall_7 {
   class=MaterialExpressionMaterialFunctionCall
   inputs [
    BaseNormal {from=MaterialExpressionNamedRerouteUsage_31}
    AdditionalNormal {from=MaterialExpressionLinearInterpolate_14}
   ]
   properties {materialFunction=&BlendAngleCorrectedNormals}
  }
  MaterialExpressionMaterialFunctionCall_8 {
   class=MaterialExpressionMaterialFunctionCall
   inputs [Color {from=MaterialExpressionReroute_11, output=SkinMaskTexture}]
   properties {materialFunction=&SplitComponents}
  }
  MaterialExpressionMaterialFunctionCall_9 {
   class=MaterialExpressionMaterialFunctionCall
   inputs [
    Normal {from=MaterialExpressionTextureSampleParameter2D_7, output=RGB}
    Flatness {from=MaterialExpressionReroute_40}
   ]
   properties {materialFunction=&FlattenNormal}
  }
  MaterialExpressionMultiply_0 {
   class=MaterialExpressionMultiply
   inputs [
    A {from=MaterialExpressionReroute_49, output=Result}
    B {from=MaterialExpressionNamedRerouteUsage_1}
   ]
  }
  MaterialExpressionMultiply_1 {
   class=MaterialExpressionMultiply
   inputs [
    A {from=MaterialExpressionReroute_47, output=BaseColorInput}
    B {from=MaterialExpressionSaturate_2}
   ]
  }
  MaterialExpressionMultiply_10 {
   class=MaterialExpressionMultiply
   inputs [
    A {from=MaterialExpressionTime_0}
    B {from=MaterialExpressionScalarParameter_18}
   ]
  }
  MaterialExpressionMultiply_11 {
   class=MaterialExpressionMultiply
   inputs [
    A {from=MaterialExpressionTextureCoordinate_8}
    B {from=MaterialExpressionAppendVector_2}
   ]
  }
  MaterialExpressionMultiply_12 {
   class=MaterialExpressionMultiply
   inputs [
    A {from=MaterialExpressionAppendVector_3}
    B {from=MaterialExpressionTime_1}
   ]
  }
  MaterialExpressionMultiply_13 {
   class=MaterialExpressionMultiply
   inputs [
    A {from=MaterialExpressionConstant_2}
    B {from=MaterialExpressionScalarParameter_19}
   ]
  }
  MaterialExpressionMultiply_14 {
   class=MaterialExpressionMultiply
   inputs [
    A {from=MaterialExpressionMultiply_13}
    B {from=MaterialExpressionAppendVector_5}
   ]
  }
  MaterialExpressionMultiply_15 {
   class=MaterialExpressionMultiply
   inputs [
    A {from=MaterialExpressionScalarParameter_20}
    B {from=MaterialExpressionStaticSwitchParameter_15}
   ]
  }
  MaterialExpressionMultiply_16 {
   class=MaterialExpressionMultiply
   inputs [
    A {from=MaterialExpressionScalarParameter_21}
    B {from=MaterialExpressionTime_2}
   ]
  }
  MaterialExpressionMultiply_17 {
   class=MaterialExpressionMultiply
   inputs [
    A {from=MaterialExpressionScalarParameter_23}
    B {from=MaterialExpressionLinearInterpolate_20}
   ]
  }
  MaterialExpressionMultiply_18 {
   class=MaterialExpressionMultiply
   inputs [
    A {from=MaterialExpressionStaticSwitchParameter_17}
    B {from=MaterialExpressionStaticSwitchParameter_14}
   ]
  }
  MaterialExpressionMultiply_19 {
   class=MaterialExpressionMultiply
   inputs [
    A {from=MaterialExpressionTextureCoordinate_9}
    B {from=MaterialExpressionAppendVector_6}
   ]
  }
  MaterialExpressionMultiply_2 {
   class=MaterialExpressionMultiply
   inputs [
    A {from=MaterialExpressionMultiply_0}
    B {from=MaterialExpressionNamedRerouteUsage_24}
   ]
  }
  MaterialExpressionMultiply_20 {
   class=MaterialExpressionMultiply
   inputs [
    A {from=MaterialExpressionAppendVector_7}
    B {from=MaterialExpressionTime_3}
   ]
  }
  MaterialExpressionMultiply_21 {
   class=MaterialExpressionMultiply
   inputs [
    A {from=MaterialExpressionStaticSwitchParameter_31}
    B {from=MaterialExpressionAppendVector_11}
   ]
  }
  MaterialExpressionMultiply_22 {
   class=MaterialExpressionMultiply
   inputs [
    A {from=MaterialExpressionConstant_10}
    B {from=MaterialExpressionScalarParameter_27}
   ]
  }
  MaterialExpressionMultiply_23 {
   class=MaterialExpressionMultiply
   inputs [
    A {from=MaterialExpressionMultiply_22}
    B {from=MaterialExpressionAppendVector_10}
   ]
  }
  MaterialExpressionMultiply_24 {
   class=MaterialExpressionMultiply
   inputs [
    A {from=MaterialExpressionTime_5}
    B {from=MaterialExpressionScalarParameter_28}
   ]
  }
  MaterialExpressionMultiply_25 {
   class=MaterialExpressionMultiply
   inputs [
    A {from=MaterialExpressionScalarParameter_30}
    B {from=MaterialExpressionStaticSwitchParameter_27}
   ]
  }
  MaterialExpressionMultiply_26 {
   class=MaterialExpressionMultiply
   inputs [
    A {from=MaterialExpressionScalarParameter_31}
    B {from=MaterialExpressionTime_6}
   ]
  }
  MaterialExpressionMultiply_27 {
   class=MaterialExpressionMultiply
   inputs [
    A {from=MaterialExpressionScalarParameter_33}
    B {from=MaterialExpressionLinearInterpolate_26}
   ]
  }
  MaterialExpressionMultiply_28 {
   class=MaterialExpressionMultiply
   inputs [
    A {from=MaterialExpressionStaticSwitchParameter_29}
    B {from=MaterialExpressionStaticSwitchParameter_26}
   ]
  }
  MaterialExpressionMultiply_29 {
   class=MaterialExpressionMultiply
   inputs [
    A {from=MaterialExpressionScalarParameter_26}
    B {from=MaterialExpressionTextureSampleParameter2D_11, output=RGB}
   ]
  }
  MaterialExpressionMultiply_3 {
   class=MaterialExpressionMultiply
   inputs [
    A {from=MaterialExpressionTextureCoordinate_0}
    B {from=MaterialExpressionScalarParameter_15}
   ]
  }
  MaterialExpressionMultiply_38 {
   class=MaterialExpressionMultiply
   inputs [
    A {from=MaterialExpressionTextureCoordinate_16}
    B {from=MaterialExpressionAppendVector_15}
   ]
  }
  MaterialExpressionMultiply_39 {
   class=MaterialExpressionMultiply
   inputs [
    A {from=MaterialExpressionAppendVector_16}
    B {from=MaterialExpressionTime_10}
   ]
  }
  MaterialExpressionMultiply_4 {
   class=MaterialExpressionMultiply
   inputs [
    A {from=MaterialExpressionMultiply_6}
    B {from=MaterialExpressionReroute_60}
   ]
  }
  MaterialExpressionMultiply_40 {
   class=MaterialExpressionMultiply
   inputs [
    A {from=MaterialExpressionConstant_21}
    B {from=MaterialExpressionScalarParameter_41}
   ]
  }
  MaterialExpressionMultiply_41 {
   class=MaterialExpressionMultiply
   inputs [
    A {from=MaterialExpressionMultiply_40}
    B {from=MaterialExpressionAppendVector_18}
   ]
  }
  MaterialExpressionMultiply_42 {
   class=MaterialExpressionMultiply
   inputs [
    A {from=MaterialExpressionTime_11}
    B {from=MaterialExpressionScalarParameter_42}
   ]
  }
  MaterialExpressionMultiply_43 {
   class=MaterialExpressionMultiply
   inputs [
    A {from=MaterialExpressionScalarParameter_44}
    B {from=MaterialExpressionStaticSwitchParameter_49}
   ]
  }
  MaterialExpressionMultiply_44 {
   class=MaterialExpressionMultiply
   inputs [
    A {from=MaterialExpressionScalarParameter_45}
    B {from=MaterialExpressionTime_12}
   ]
  }
  MaterialExpressionMultiply_45 {
   class=MaterialExpressionMultiply
   inputs [
    A {from=MaterialExpressionScalarParameter_47}
    B {from=MaterialExpressionLinearInterpolate_36}
   ]
  }
  MaterialExpressionMultiply_46 {
   class=MaterialExpressionMultiply
   inputs [
    A {from=MaterialExpressionStaticSwitchParameter_51}
    B {from=MaterialExpressionStaticSwitchParameter_48}
   ]
  }
  MaterialExpressionMultiply_47 {
   class=MaterialExpressionMultiply
   inputs [
    A {from=MaterialExpressionTextureCoordinate_19}
    B {from=MaterialExpressionAppendVector_19}
   ]
  }
  MaterialExpressionMultiply_48 {
   class=MaterialExpressionMultiply
   inputs [
    A {from=MaterialExpressionAppendVector_20}
    B {from=MaterialExpressionTime_13}
   ]
  }
  MaterialExpressionMultiply_49 {
   class=MaterialExpressionMultiply
   inputs [
    A {from=MaterialExpressionConstant_27}
    B {from=MaterialExpressionScalarParameter_49}
   ]
  }
  MaterialExpressionMultiply_5 {
   class=MaterialExpressionMultiply
   inputs [
    A {from=MaterialExpressionTextureCoordinate_1}
    B {from=MaterialExpressionScalarParameter_0}
   ]
  }
  MaterialExpressionMultiply_50 {
   class=MaterialExpressionMultiply
   inputs [
    A {from=MaterialExpressionMultiply_49}
    B {from=MaterialExpressionAppendVector_22}
   ]
  }
  MaterialExpressionMultiply_51 {
   class=MaterialExpressionMultiply
   inputs [
    A {from=MaterialExpressionTime_14}
    B {from=MaterialExpressionScalarParameter_50}
   ]
  }
  MaterialExpressionMultiply_52 {
   class=MaterialExpressionMultiply
   inputs [
    A {from=MaterialExpressionScalarParameter_52}
    B {from=MaterialExpressionStaticSwitchParameter_61}
   ]
  }
  MaterialExpressionMultiply_53 {
   class=MaterialExpressionMultiply
   inputs [
    A {from=MaterialExpressionScalarParameter_53}
    B {from=MaterialExpressionTime_15}
   ]
  }
  MaterialExpressionMultiply_54 {
   class=MaterialExpressionMultiply
   inputs [
    A {from=MaterialExpressionScalarParameter_55}
    B {from=MaterialExpressionLinearInterpolate_41}
   ]
  }
  MaterialExpressionMultiply_55 {
   class=MaterialExpressionMultiply
   inputs [
    A {from=MaterialExpressionStaticSwitchParameter_63}
    B {from=MaterialExpressionStaticSwitchParameter_60}
   ]
  }
  MaterialExpressionMultiply_6 {
   class=MaterialExpressionMultiply
   inputs [
    A {from=MaterialExpressionChannelMaskParameter_1, output=R}
    B {from=MaterialExpressionScalarParameter_16}
   ]
  }
  MaterialExpressionMultiply_7 {
   class=MaterialExpressionMultiply
   inputs [
    A {from=MaterialExpressionNamedRerouteUsage_6}
    B {from=MaterialExpressionReroute_57, output=Output}
   ]
  }
  MaterialExpressionMultiply_8 {
   class=MaterialExpressionMultiply
   inputs [
    A {from=MaterialExpressionSubtract_1}
    B {from=""}
   ]
   properties {constB=1.1087}
  }
  MaterialExpressionMultiply_9 {
   class=MaterialExpressionMultiply
   inputs [
    A {from=MaterialExpressionStaticSwitchParameter_8}
    B {from=MaterialExpressionScalarParameter_17}
   ]
  }
  MaterialExpressionNamedRerouteDeclaration_0 {
   class=MaterialExpressionNamedRerouteDeclaration
   inputs [None {from=MaterialExpressionNamedRerouteUsage_2}]
   properties {name=MetallicMasked, nodeColor=<rgba 0 1 0.7656 1>}
  }
  MaterialExpressionNamedRerouteDeclaration_1 {
   class=MaterialExpressionNamedRerouteDeclaration
   inputs [None {from=MaterialExpressionStaticSwitchParameter_3}]
   properties {name=UV, nodeColor=<rgba 1 0.8646 0.8646 1>}
  }
  MaterialExpressionNamedRerouteDeclaration_10 {
   class=MaterialExpressionNamedRerouteDeclaration
   inputs [None {from=MaterialExpressionReroute_45, output=SharpMasksSwitchOutput}]
   properties {name=SkinMaskTexture, nodeColor=<rgba 0.6038 0.3325 0.013 1>}
  }
  MaterialExpressionNamedRerouteDeclaration_11 {
   class=MaterialExpressionNamedRerouteDeclaration
   inputs [None {from=MaterialExpressionReroute_5, output=RoughnessMasked}]
   properties {name=RoughnessMasked, nodeColor=<rgba 0 1 0.7656 1>}
  }
  MaterialExpressionNamedRerouteDeclaration_12 {
   class=MaterialExpressionNamedRerouteDeclaration
   inputs [None {from=MaterialExpressionReroute_23, output=TeamColor}]
   properties {name=TeamColor, nodeColor=<rgba 0.9219 0 1 1>}
  }
  MaterialExpressionNamedRerouteDeclaration_13 {
   class=MaterialExpressionNamedRerouteDeclaration
   inputs [None {from=MaterialExpressionReroute_24, output=AccentColor}]
   properties {name=AccentColor, nodeColor=<rgba 0.9219 0 1 1>}
  }
  MaterialExpressionNamedRerouteDeclaration_14 {
   class=MaterialExpressionNamedRerouteDeclaration
   inputs [None {from=MaterialExpressionReroute_25, output=PaintColor}]
   properties {name=PaintColor, nodeColor=<rgba 0.9219 0 1 1>}
  }
  MaterialExpressionNamedRerouteDeclaration_15 {
   class=MaterialExpressionNamedRerouteDeclaration
   inputs [None {from=MaterialExpressionReroute_15, output=BottomNormalBase}]
   properties {name=BottomNormalBaseTexture, nodeColor=<rgba 0.6038 0.3325 0.013 1>}
  }
  MaterialExpressionNamedRerouteDeclaration_16 {
   class=MaterialExpressionNamedRerouteDeclaration
   inputs [None {from=MaterialExpressionReroute_7, output=BaseColor}]
   properties {name=BaseColorTexture, nodeColor=<rgba 0.6038 0.3325 0.013 1>}
  }
  MaterialExpressionNamedRerouteDeclaration_17 {
   class=MaterialExpressionNamedRerouteDeclaration
   inputs [None {from=MaterialExpressionAdd_2}]
   properties {name=BaseColor, nodeColor=<rgba 0 1 0.7656 1>}
  }
  MaterialExpressionNamedRerouteDeclaration_18 {
   class=MaterialExpressionNamedRerouteDeclaration
   inputs [None {from=MaterialExpressionNamedRerouteUsage_20}]
   properties {name=CCThicknessMasked, nodeColor=<rgba 0 1 0.7656 1>}
  }
  MaterialExpressionNamedRerouteDeclaration_19 {
   class=MaterialExpressionNamedRerouteDeclaration
   inputs [None {from=MaterialExpressionNamedRerouteUsage_17}]
   properties {name=CCRoughnessMasked, nodeColor=<rgba 0 1 0.7656 1>}
  }
  MaterialExpressionNamedRerouteDeclaration_2 {
   class=MaterialExpressionNamedRerouteDeclaration
   inputs [None {from=MaterialExpressionReroute_6, output=Metallic}]
   properties {name=Metallic, nodeColor=<rgba 0 0.1434 1 1>}
  }
  MaterialExpressionNamedRerouteDeclaration_20 {
   class=MaterialExpressionNamedRerouteDeclaration
   inputs [None {from=MaterialExpressionReroute_38, output=BottomNormalAccent}]
   properties {name=BottomNormalAccentTexture, nodeColor=<rgba 0.6038 0.3325 0.013 1>}
  }
  MaterialExpressionNamedRerouteDeclaration_21 {
   class=MaterialExpressionNamedRerouteDeclaration
   inputs [None {from=MaterialExpressionReroute_19, output=AO}]
   properties {name=AOTexture, nodeColor=<rgba 0.6038 0.3325 0.013 1>}
  }
  MaterialExpressionNamedRerouteDeclaration_22 {
   class=MaterialExpressionNamedRerouteDeclaration
   inputs [None {from=MaterialExpressionReroute_14, output=BodyNormalLOD}]
   properties {name=BodyNormalLOD, nodeColor=<rgba 0.6038 0.3325 0.013 1>}
  }
  MaterialExpressionNamedRerouteDeclaration_23 {
   class=MaterialExpressionNamedRerouteDeclaration
   inputs [None {from=MaterialExpressionReroute_37, output=ClearCoatBottomNormal}]
   properties {name=CCBottomNormal, nodeColor=<rgba 0 0.1434 1 1>}
  }
  MaterialExpressionNamedRerouteDeclaration_24 {
   class=MaterialExpressionNamedRerouteDeclaration
   inputs [None {from=MaterialExpressionReroute_53, output=PBR}]
   properties {name=BodyTextures_RoughnessMetallic, nodeColor=<rgba 1 0.3516 0 1>}
  }
  MaterialExpressionNamedRerouteDeclaration_25 {
   class=MaterialExpressionNamedRerouteDeclaration
   inputs [None {from=MaterialExpressionMultiply_3}]
   properties {name=UV_BottomNormal, nodeColor=<rgba 0.6458 0.6458 0.6458 1>}
  }
  MaterialExpressionNamedRerouteDeclaration_26 {
   class=MaterialExpressionNamedRerouteDeclaration
   inputs [None {from=MaterialExpressionReroute_46, output=TextureRawOutput}]
   properties {name=SkinMaskTextureRaw, nodeColor=<rgba 1 0 0.2812 1>}
  }
  MaterialExpressionNamedRerouteDeclaration_27 {
   class=MaterialExpressionNamedRerouteDeclaration
   inputs [None {from=MaterialExpressionAdd_10}]
   properties {name=EmissiveOutput, nodeColor=<rgba 0 0.9297 1 1>}
  }
  MaterialExpressionNamedRerouteDeclaration_28 {
   class=MaterialExpressionNamedRerouteDeclaration
   inputs [None {from=MaterialExpressionStaticSwitchParameter_5}]
   properties {name=MetallicOutput, nodeColor=<rgba 0 1 0.7647 1>}
  }
  MaterialExpressionNamedRerouteDeclaration_29 {
   class=MaterialExpressionNamedRerouteDeclaration
   inputs [None {from=MaterialExpressionStaticSwitchParameter_6}]
   properties {name=RoughnessOutput, nodeColor=<rgba 0 1 0.7647 1>}
  }
  MaterialExpressionNamedRerouteDeclaration_3 {
   class=MaterialExpressionNamedRerouteDeclaration
   inputs [None {from=MaterialExpressionReroute_9, output=Roughness}]
   properties {name=Roughness, nodeColor=<rgba 0 0.1434 1 1>}
  }
  MaterialExpressionNamedRerouteDeclaration_30 {
   class=MaterialExpressionNamedRerouteDeclaration
   inputs [None {from=MaterialExpressionReroute_54, output=EmissiveColor}]
   properties {name=EmissiveColor, nodeColor=<rgba 0.9219 0 1 1>}
  }
  MaterialExpressionNamedRerouteDeclaration_32 {
   class=MaterialExpressionNamedRerouteDeclaration
   inputs [None {from=MaterialExpressionReroute_22, output=Curvature}]
   properties {name=Curvature, nodeColor=<rgba 0.6038 0.3325 0.013 1>}
  }
  MaterialExpressionNamedRerouteDeclaration_4 {
   class=MaterialExpressionNamedRerouteDeclaration
   inputs [None {from=MaterialExpressionReroute_32, output=ClearCoat Thickness}]
   properties {name=CCThickness, nodeColor=<rgba 0 0.1434 1 1>}
  }
  MaterialExpressionNamedRerouteDeclaration_5 {
   class=MaterialExpressionNamedRerouteDeclaration
   inputs [None {from=MaterialExpressionReroute_33, output=ClearCoat Roughness}]
   properties {name=CCRoughness, nodeColor=<rgba 0 0.1434 1 1>}
  }
  MaterialExpressionNamedRerouteDeclaration_6 {
   class=MaterialExpressionNamedRerouteDeclaration
   inputs [None {from=MaterialExpressionReroute_0, output=DetailBase}]
   properties {name=DetailBaseTexture, nodeColor=<rgba 0.6038 0.3325 0.013 1>}
  }
  MaterialExpressionNamedRerouteDeclaration_7 {
   class=MaterialExpressionNamedRerouteDeclaration
   inputs [None {from=MaterialExpressionReroute_1, output=DetailAccent}]
   properties {name=DetailAccentTexture, nodeColor=<rgba 0.6038 0.3325 0.013 1>}
  }
  MaterialExpressionNamedRerouteDeclaration_8 {
   class=MaterialExpressionNamedRerouteDeclaration
   inputs [None {from=MaterialExpressionReroute_17, output=DetailsMasked}]
   properties {name=DetailsMasked, nodeColor=<rgba 0 1 0.7656 1>}
  }
  MaterialExpressionNamedRerouteDeclaration_9 {
   class=MaterialExpressionNamedRerouteDeclaration
   inputs [None {from=MaterialExpressionMultiply_5}]
   properties {name=UV_Detail, nodeColor=<rgba 1 0.8646 0.8646 1>}
  }
  MaterialExpressionNamedRerouteUsage_0 {class=MaterialExpressionNamedRerouteUsage}
  MaterialExpressionNamedRerouteUsage_1 {class=MaterialExpressionNamedRerouteUsage}
  MaterialExpressionNamedRerouteUsage_10 {class=MaterialExpressionNamedRerouteUsage}
  MaterialExpressionNamedRerouteUsage_11 {class=MaterialExpressionNamedRerouteUsage}
  MaterialExpressionNamedRerouteUsage_12 {class=MaterialExpressionNamedRerouteUsage}
  MaterialExpressionNamedRerouteUsage_13 {class=MaterialExpressionNamedRerouteUsage}
  MaterialExpressionNamedRerouteUsage_14 {class=MaterialExpressionNamedRerouteUsage}
  MaterialExpressionNamedRerouteUsage_15 {class=MaterialExpressionNamedRerouteUsage}
  MaterialExpressionNamedRerouteUsage_16 {class=MaterialExpressionNamedRerouteUsage}
  MaterialExpressionNamedRerouteUsage_17 {class=MaterialExpressionNamedRerouteUsage}
  MaterialExpressionNamedRerouteUsage_18 {class=MaterialExpressionNamedRerouteUsage}
  MaterialExpressionNamedRerouteUsage_19 {class=MaterialExpressionNamedRerouteUsage}
  MaterialExpressionNamedRerouteUsage_2 {class=MaterialExpressionNamedRerouteUsage}
  MaterialExpressionNamedRerouteUsage_20 {class=MaterialExpressionNamedRerouteUsage}
  MaterialExpressionNamedRerouteUsage_21 {class=MaterialExpressionNamedRerouteUsage}
  MaterialExpressionNamedRerouteUsage_22 {class=MaterialExpressionNamedRerouteUsage}
  MaterialExpressionNamedRerouteUsage_23 {class=MaterialExpressionNamedRerouteUsage}
  MaterialExpressionNamedRerouteUsage_24 {class=MaterialExpressionNamedRerouteUsage}
  MaterialExpressionNamedRerouteUsage_25 {class=MaterialExpressionNamedRerouteUsage}
  MaterialExpressionNamedRerouteUsage_26 {class=MaterialExpressionNamedRerouteUsage}
  MaterialExpressionNamedRerouteUsage_27 {class=MaterialExpressionNamedRerouteUsage}
  MaterialExpressionNamedRerouteUsage_28 {class=MaterialExpressionNamedRerouteUsage}
  MaterialExpressionNamedRerouteUsage_29 {class=MaterialExpressionNamedRerouteUsage}
  MaterialExpressionNamedRerouteUsage_3 {class=MaterialExpressionNamedRerouteUsage}
  MaterialExpressionNamedRerouteUsage_30 {class=MaterialExpressionNamedRerouteUsage}
  MaterialExpressionNamedRerouteUsage_31 {class=MaterialExpressionNamedRerouteUsage}
  MaterialExpressionNamedRerouteUsage_32 {class=MaterialExpressionNamedRerouteUsage}
  MaterialExpressionNamedRerouteUsage_33 {class=MaterialExpressionNamedRerouteUsage}
  MaterialExpressionNamedRerouteUsage_34 {class=MaterialExpressionNamedRerouteUsage}
  MaterialExpressionNamedRerouteUsage_35 {class=MaterialExpressionNamedRerouteUsage}
  MaterialExpressionNamedRerouteUsage_36 {class=MaterialExpressionNamedRerouteUsage}
  MaterialExpressionNamedRerouteUsage_37 {class=MaterialExpressionNamedRerouteUsage}
  MaterialExpressionNamedRerouteUsage_38 {class=MaterialExpressionNamedRerouteUsage}
  MaterialExpressionNamedRerouteUsage_39 {class=MaterialExpressionNamedRerouteUsage}
  MaterialExpressionNamedRerouteUsage_4 {class=MaterialExpressionNamedRerouteUsage}
  MaterialExpressionNamedRerouteUsage_5 {class=MaterialExpressionNamedRerouteUsage}
  MaterialExpressionNamedRerouteUsage_6 {class=MaterialExpressionNamedRerouteUsage}
  MaterialExpressionNamedRerouteUsage_7 {class=MaterialExpressionNamedRerouteUsage}
  MaterialExpressionNamedRerouteUsage_8 {class=MaterialExpressionNamedRerouteUsage}
  MaterialExpressionNamedRerouteUsage_9 {class=MaterialExpressionNamedRerouteUsage}
  MaterialExpressionOneMinus_0 {class=MaterialExpressionOneMinus, inputs [None {from=MaterialExpressionStaticComponentMaskParameter_0}]}
  MaterialExpressionOneMinus_1 {class=MaterialExpressionOneMinus, inputs [None {from=MaterialExpressionStaticComponentMaskParameter_1}]}
  MaterialExpressionOneMinus_2 {class=MaterialExpressionOneMinus, inputs [None {from=MaterialExpressionComponentMask_6}]}
  MaterialExpressionOneMinus_3 {class=MaterialExpressionOneMinus, inputs [None {from=MaterialExpressionStaticComponentMaskParameter_3}]}
  MaterialExpressionOneMinus_4 {class=MaterialExpressionOneMinus, inputs [None {from=MaterialExpressionStaticComponentMaskParameter_4}]}
  MaterialExpressionPinBase_0 {
   class=MaterialExpressionPinBase
   inputs []
   properties {reroutePins [SkinMaskTexture {expression=&self:MaterialExpressionReroute_10}]}
  }
  MaterialExpressionPinBase_1 {
   class=MaterialExpressionPinBase
   inputs [
    Metallic {from=MaterialExpressionLinearInterpolate_1, output=""}
    Roughness {from=MaterialExpressionLinearInterpolate_3, output=""}
   ]
   properties {reroutePins [
     Metallic {expression=&self:MaterialExpressionReroute_6}
     Roughness {expression=&self:MaterialExpressionReroute_9}
    ]}
  }
  MaterialExpressionPinBase_10 {
   class=MaterialExpressionPinBase
   inputs []
   properties {reroutePins [
     BaseColor {expression=&self:MaterialExpressionReroute_27}
     BaseColorTexture {expression=&self:MaterialExpressionReroute_29}
     SkinMaskTexture {expression=&self:MaterialExpressionReroute_30}
    ]}
  }
  MaterialExpressionPinBase_11 {
   class=MaterialExpressionPinBase
   inputs [BaseColor {from=MaterialExpressionStaticSwitchParameter_0, output=""}]
   properties {reroutePins [BaseColor {expression=&self:MaterialExpressionReroute_28}]}
  }
  MaterialExpressionPinBase_12 {
   class=MaterialExpressionPinBase
   inputs []
   properties {reroutePins [
     SkinMaskTexture {expression=&self:MaterialExpressionReroute_11}
     BottomNormalBaseTexture {expression=&self:MaterialExpressionReroute_35}
     BottomNormalAccentTexture {expression=&self:MaterialExpressionReroute_36}
    ]}
  }
  MaterialExpressionPinBase_13 {
   class=MaterialExpressionPinBase
   inputs [
    ClearCoat Thickness {from=MaterialExpressionLinearInterpolate_12, output=""}
    ClearCoat Roughness {from=MaterialExpressionLinearInterpolate_18, output=""}
    ClearCoatBottomNormal {from=MaterialExpressionMaterialFunctionCall_7, output=Result}
   ]
   properties {reroutePins [
     ClearCoat Thickness {expression=&self:MaterialExpressionReroute_32}
     ClearCoat Roughness {expression=&self:MaterialExpressionReroute_33}
     ClearCoatBottomNormal {expression=&self:MaterialExpressionReroute_37}
    ]}
  }
  MaterialExpressionPinBase_14 {
   class=MaterialExpressionPinBase
   inputs []
   properties {reroutePins [
     DetailBaseTexture {expression=&self:MaterialExpressionReroute_31}
     DetailAccentTexture {expression=&self:MaterialExpressionReroute_34}
     SkinMaskTexture {expression=&self:MaterialExpressionReroute_16}
    ]}
  }
  MaterialExpressionPinBase_15 {
   class=MaterialExpressionPinBase
   inputs [DetailsMasked {from=MaterialExpressionLinearInterpolate_15, output=""}]
   properties {reroutePins [DetailsMasked {expression=&self:MaterialExpressionReroute_17}]}
  }
  MaterialExpressionPinBase_16 {
   class=MaterialExpressionPinBase
   inputs []
   properties {reroutePins [RGB {expression=&self:MaterialExpressionReroute_42}]}
  }
  MaterialExpressionPinBase_17 {
   class=MaterialExpressionPinBase
   inputs [
    SharpMasksSwitchOutput {from=MaterialExpressionStaticSwitchParameter_1, output=""}
    TextureRawOutput {from=MaterialExpressionReroute_42, output=RGB}
   ]
   properties {reroutePins [
     SharpMasksSwitchOutput {expression=&self:MaterialExpressionReroute_45}
     TextureRawOutput {expression=&self:MaterialExpressionReroute_46}
    ]}
  }
  MaterialExpressionPinBase_18 {
   class=MaterialExpressionPinBase
   inputs []
   properties {reroutePins [
     BaseColorInput {expression=&self:MaterialExpressionReroute_47}
     TextureInput {expression=&self:MaterialExpressionReroute_48}
    ]}
  }
  MaterialExpressionPinBase_19 {
   class=MaterialExpressionPinBase
   inputs [Result {from=MaterialExpressionStaticSwitchParameter_2, output=""}]
   properties {reroutePins [Result {expression=&self:MaterialExpressionReroute_49}]}
  }
  MaterialExpressionPinBase_2 {
   class=MaterialExpressionPinBase
   inputs []
   properties {reroutePins [
     UV {expression=&self:MaterialExpressionReroute_4}
     UV Detail {expression=&self:MaterialExpressionReroute_12}
     UV BottomNormal {expression=&self:MaterialExpressionReroute_41}
     LOD {expression=&self:MaterialExpressionReroute_21}
     BottomNormalFlatness {expression=&self:MaterialExpressionReroute_39}
    ]}
  }
  MaterialExpressionPinBase_20 {
   class=MaterialExpressionPinBase
   inputs []
   properties {reroutePins [
     SkinMaskTexture {expression=&self:MaterialExpressionReroute_55}
     BodyTextures {expression=&self:MaterialExpressionReroute_58}
     EmissiveColor {expression=&self:MaterialExpressionReroute_56}
    ]}
  }
  MaterialExpressionPinBase_21 {
   class=MaterialExpressionPinBase
   inputs [Output {from=MaterialExpressionMultiply_4, output=""}]
   properties {reroutePins [Output {expression=&self:MaterialExpressionReroute_57}]}
  }
  MaterialExpressionPinBase_3 {
   class=MaterialExpressionPinBase
   inputs [
    Curvature {from=MaterialExpressionTextureSampleParameter2D_0, output=R}
    AO {from=MaterialExpressionTextureSampleParameter2D_0, output=R}
    PBR {from=MaterialExpressionAppendVector_0, output=""}
    SkinMask {from=MaterialExpressionTextureSampleParameter2D_1, output=RGB}
    BaseColor {from=MaterialExpressionTextureSampleParameter2D_2, output=RGB}
    DetailBase {from=MaterialExpressionTextureSampleParameter2D_3, output=RGB}
    DetailAccent {from=MaterialExpressionTextureSampleParameter2D_4, output=RGB}
    BodyNormalLOD {from=MaterialExpressionLinearInterpolate_4, output=""}
    BottomNormalBase {from=MaterialExpressionMaterialFunctionCall_1, output=Result}
    BottomNormalAccent {from=MaterialExpressionMaterialFunctionCall_9, output=Result}
   ]
   properties {reroutePins [
     Curvature {expression=&self:MaterialExpressionReroute_22}
     AO {expression=&self:MaterialExpressionReroute_19}
     PBR {expression=&self:MaterialExpressionReroute_53}
     SkinMask {expression=&self:MaterialExpressionReroute_44}
     BaseColor {expression=&self:MaterialExpressionReroute_7}
     DetailBase {expression=&self:MaterialExpressionReroute_0}
     DetailAccent {expression=&self:MaterialExpressionReroute_1}
     BodyNormalLOD {expression=&self:MaterialExpressionReroute_14}
     BottomNormalBase {expression=&self:MaterialExpressionReroute_15}
     BottomNormalAccent {expression=&self:MaterialExpressionReroute_38}
    ]}
  }
  MaterialExpressionPinBase_4 {class=MaterialExpressionPinBase, inputs [], properties {}}
  MaterialExpressionPinBase_5 {
   class=MaterialExpressionPinBase
   inputs [
    TeamColor {from=MaterialExpressionVectorParameter_0, output=RGB}
    AccentColor {from=MaterialExpressionVectorParameter_1, output=RGB}
    PaintColor {from=MaterialExpressionVectorParameter_3, output=RGB}
    TrimColor {from=MaterialExpressionVectorParameter_2, output=RGB}
    EmissiveColor {from=MaterialExpressionVectorParameter_5, output=RGB}
   ]
   properties {reroutePins [
     TeamColor {expression=&self:MaterialExpressionReroute_23}
     AccentColor {expression=&self:MaterialExpressionReroute_24}
     PaintColor {expression=&self:MaterialExpressionReroute_25}
     TrimColor {expression=&self:MaterialExpressionReroute_43}
     EmissiveColor {expression=&self:MaterialExpressionReroute_54}
    ]}
  }
  MaterialExpressionPinBase_6 {
   class=MaterialExpressionPinBase
   inputs []
   properties {reroutePins [
     Roughness {expression=&self:MaterialExpressionReroute_2}
     SkinMask {expression=&self:MaterialExpressionReroute_3}
    ]}
  }
  MaterialExpressionPinBase_7 {
   class=MaterialExpressionPinBase
   inputs [RoughnessMasked {from=MaterialExpressionLinearInterpolate_6, output=""}]
   properties {reroutePins [RoughnessMasked {expression=&self:MaterialExpressionReroute_5}]}
  }
  MaterialExpressionPinBase_8 {
   class=MaterialExpressionPinBase
   inputs []
   properties {reroutePins [
     TeamColor {expression=&self:MaterialExpressionReroute_8}
     AccentColor {expression=&self:MaterialExpressionReroute_13}
     PaintColor {expression=&self:MaterialExpressionReroute_18}
     SkinMaskTexture {expression=&self:MaterialExpressionReroute_20}
    ]}
  }
  MaterialExpressionPinBase_9 {
   class=MaterialExpressionPinBase
   inputs [BaseColor {from=MaterialExpressionLinearInterpolate_8, output=""}]
   properties {reroutePins [BaseColor {expression=&self:MaterialExpressionReroute_26}]}
  }
  MaterialExpressionPower_0 {
   class=MaterialExpressionPower
   inputs [
    Base {from=MaterialExpressionReroute_42, output=RGB}
    Exp {from=""}
   ]
   properties {constExponent=0.4545}
  }
  MaterialExpressionPower_1 {
   class=MaterialExpressionPower
   inputs [
    Base {from=MaterialExpressionReroute_48, output=TextureInput}
    Exp {from=""}
   ]
   properties {constExponent=0.4545}
  }
  MaterialExpressionPower_5 {
   class=MaterialExpressionPower
   inputs [
    Base {from=MaterialExpressionReroute_52}
    Exp {from=""}
   ]
   properties {constExponent=8}
  }
  MaterialExpressionPower_7 {
   class=MaterialExpressionPower
   inputs [
    Base {from=MaterialExpressionAdd_4}
    Exp {from=""}
   ]
   properties {constExponent=0.25}
  }
  MaterialExpressionReroute_10 {class=MaterialExpressionReroute, inputs [None {from=MaterialExpressionNamedRerouteUsage_12}]}
  MaterialExpressionReroute_11 {class=MaterialExpressionReroute, inputs [None {from=MaterialExpressionNamedRerouteUsage_13}]}
  MaterialExpressionReroute_12 {class=MaterialExpressionReroute, inputs [None {from=MaterialExpressionNamedRerouteUsage_26}]}
  MaterialExpressionReroute_13 {class=MaterialExpressionReroute, inputs [None {from=MaterialExpressionNamedRerouteUsage_9}]}
  MaterialExpressionReroute_16 {class=MaterialExpressionReroute, inputs [None {from=MaterialExpressionNamedRerouteUsage_18}]}
  MaterialExpressionReroute_17 {class=MaterialExpressionReroute, inputs [None {from=MaterialExpressionLinearInterpolate_15}]}
  MaterialExpressionReroute_18 {class=MaterialExpressionReroute, inputs [None {from=MaterialExpressionNamedRerouteUsage_10}]}
  MaterialExpressionReroute_2 {class=MaterialExpressionReroute, inputs [None {from=MaterialExpressionNamedRerouteUsage_5}]}
  MaterialExpressionReroute_20 {class=MaterialExpressionReroute, inputs [None {from=MaterialExpressionNamedRerouteUsage_7}]}
  MaterialExpressionReroute_21 {class=MaterialExpressionReroute, inputs [None {from=MaterialExpressionScalarParameter_1}]}
  MaterialExpressionReroute_23 {class=MaterialExpressionReroute, inputs [None {from=MaterialExpressionVectorParameter_0, output=RGB}]}
  MaterialExpressionReroute_24 {class=MaterialExpressionReroute, inputs [None {from=MaterialExpressionVectorParameter_1, output=RGB}]}
  MaterialExpressionReroute_25 {class=MaterialExpressionReroute, inputs [None {from=MaterialExpressionVectorParameter_3, output=RGB}]}
  MaterialExpressionReroute_26 {class=MaterialExpressionReroute, inputs [None {from=MaterialExpressionLinearInterpolate_8}]}
  MaterialExpressionReroute_27 {class=MaterialExpressionReroute, inputs [None {from=MaterialExpressionReroute_26, output=BaseColor}]}
  MaterialExpressionReroute_28 {class=MaterialExpressionReroute, inputs [None {from=MaterialExpressionStaticSwitchParameter_0}]}
  MaterialExpressionReroute_29 {class=MaterialExpressionReroute, inputs [None {from=MaterialExpressionNamedRerouteUsage_15}]}
  MaterialExpressionReroute_3 {class=MaterialExpressionReroute, inputs [None {from=MaterialExpressionNamedRerouteUsage_4}]}
  MaterialExpressionReroute_30 {class=MaterialExpressionReroute, inputs [None {from=MaterialExpressionNamedRerouteUsage_14}]}
  MaterialExpressionReroute_31 {class=MaterialExpressionReroute, inputs [None {from=MaterialExpressionNamedRerouteUsage_29}]}
  MaterialExpressionReroute_32 {class=MaterialExpressionReroute, inputs [None {from=MaterialExpressionLinearInterpolate_12}]}
  MaterialExpressionReroute_33 {class=MaterialExpressionReroute, inputs [None {from=MaterialExpressionLinearInterpolate_18}]}
  MaterialExpressionReroute_34 {class=MaterialExpressionReroute, inputs [None {from=MaterialExpressionNamedRerouteUsage_30}]}
  MaterialExpressionReroute_35 {class=MaterialExpressionReroute, inputs [None {from=MaterialExpressionNamedRerouteUsage_11}]}
  MaterialExpressionReroute_36 {class=MaterialExpressionReroute, inputs [None {from=MaterialExpressionNamedRerouteUsage_19}]}
  MaterialExpressionReroute_37 {class=MaterialExpressionReroute, inputs [None {from=MaterialExpressionMaterialFunctionCall_7, output=Result}]}
  MaterialExpressionReroute_38 {class=MaterialExpressionReroute, inputs [None {from=MaterialExpressionMaterialFunctionCall_9, output=Result}]}
  MaterialExpressionReroute_39 {class=MaterialExpressionReroute, inputs [None {from=MaterialExpressionScalarParameter_14}]}
  MaterialExpressionReroute_4 {class=MaterialExpressionReroute, inputs [None {from=MaterialExpressionNamedRerouteUsage_25}]}
  MaterialExpressionReroute_40 {class=MaterialExpressionReroute, inputs [None {from=MaterialExpressionReroute_39, output=BottomNormalFlatness}]}
  MaterialExpressionReroute_41 {class=MaterialExpressionReroute, inputs [None {from=MaterialExpressionNamedRerouteUsage_21}]}
  MaterialExpressionReroute_42 {class=MaterialExpressionReroute, inputs [None {from=MaterialExpressionReroute_44, output=SkinMask}]}
  MaterialExpressionReroute_43 {class=MaterialExpressionReroute, inputs [None {from=MaterialExpressionVectorParameter_2, output=RGB}]}
  MaterialExpressionReroute_45 {class=MaterialExpressionReroute, inputs [None {from=MaterialExpressionStaticSwitchParameter_1}]}
  MaterialExpressionReroute_46 {class=MaterialExpressionReroute, inputs [None {from=MaterialExpressionReroute_42, output=RGB}]}
  MaterialExpressionReroute_47 {class=MaterialExpressionReroute, inputs [None {from=MaterialExpressionReroute_28, output=BaseColor}]}
  MaterialExpressionReroute_48 {class=MaterialExpressionReroute, inputs [None {from=MaterialExpressionNamedRerouteUsage_27}]}
  MaterialExpressionReroute_49 {class=MaterialExpressionReroute, inputs [None {from=MaterialExpressionStaticSwitchParameter_2}]}
  MaterialExpressionReroute_5 {class=MaterialExpressionReroute, inputs [None {from=MaterialExpressionLinearInterpolate_6}]}
  MaterialExpressionReroute_50 {class=MaterialExpressionReroute, inputs [None {from=MaterialExpressionReroute_47, output=BaseColorInput}]}
  MaterialExpressionReroute_51 {class=MaterialExpressionReroute, inputs [None {from=MaterialExpressionReroute_50}]}
  MaterialExpressionReroute_52 {class=MaterialExpressionReroute, inputs [None {from=MaterialExpressionComponentMask_1}]}
  MaterialExpressionReroute_53 {class=MaterialExpressionReroute, inputs [None {from=MaterialExpressionAppendVector_0}]}
  MaterialExpressionReroute_54 {class=MaterialExpressionReroute, inputs [None {from=MaterialExpressionVectorParameter_5, output=RGB}]}
  MaterialExpressionReroute_55 {class=MaterialExpressionReroute, inputs [None {from=MaterialExpressionNamedRerouteUsage_39}]}
  MaterialExpressionReroute_56 {class=MaterialExpressionReroute, inputs [None {from=MaterialExpressionNamedRerouteUsage_33}]}
  MaterialExpressionReroute_57 {class=MaterialExpressionReroute, inputs [None {from=MaterialExpressionMultiply_4}]}
  MaterialExpressionReroute_58 {class=MaterialExpressionReroute, inputs [None {from=MaterialExpressionNamedRerouteUsage_38}]}
  MaterialExpressionReroute_59 {class=MaterialExpressionReroute, inputs [None {from=MaterialExpressionReroute_56, output=EmissiveColor}]}
  MaterialExpressionReroute_6 {class=MaterialExpressionReroute, inputs [None {from=MaterialExpressionLinearInterpolate_1}]}
  MaterialExpressionReroute_60 {class=MaterialExpressionReroute, inputs [None {from=MaterialExpressionReroute_59}]}
  MaterialExpressionReroute_8 {class=MaterialExpressionReroute, inputs [None {from=MaterialExpressionNamedRerouteUsage_8}]}
  MaterialExpressionReroute_9 {class=MaterialExpressionReroute, inputs [None {from=MaterialExpressionLinearInterpolate_3}]}
  MaterialExpressionRotator_0 {
   class=MaterialExpressionRotator
   inputs [
    Coordinate {from=MaterialExpressionStaticSwitchParameter_9, output=""}
    Time {from=MaterialExpressionMultiply_10, output=""}
   ]
   properties {}
  }
  MaterialExpressionRotator_1 {
   class=MaterialExpressionRotator
   inputs [
    Coordinate {from=MaterialExpressionStaticSwitchParameter_23, output=""}
    Time {from=MaterialExpressionMultiply_24, output=""}
   ]
   properties {}
  }
  MaterialExpressionRotator_3 {
   class=MaterialExpressionRotator
   inputs [
    Coordinate {from=MaterialExpressionStaticSwitchParameter_45, output=""}
    Time {from=MaterialExpressionMultiply_42, output=""}
   ]
   properties {}
  }
  MaterialExpressionRotator_4 {
   class=MaterialExpressionRotator
   inputs [
    Coordinate {from=MaterialExpressionStaticSwitchParameter_57, output=""}
    Time {from=MaterialExpressionMultiply_51, output=""}
   ]
   properties {}
  }
  MaterialExpressionSaturate_0 {class=MaterialExpressionSaturate, inputs [None {from=MaterialExpressionConstantBiasScale_1}]}
  MaterialExpressionSaturate_1 {class=MaterialExpressionSaturate, inputs [None {from=MaterialExpressionMaterialFunctionCall_10, output=G}]}
  MaterialExpressionSaturate_2 {class=MaterialExpressionSaturate, inputs [None {from=MaterialExpressionClamp_1}]}
  MaterialExpressionSaturate_4 {class=MaterialExpressionSaturate, inputs [None {from=MaterialExpressionMultiply_8}]}
  MaterialExpressionScalarParameter_0 {class=MaterialExpressionScalarParameter, properties {defaultValue=3, group=Finish, parameterName=DetailTiling}}
  MaterialExpressionScalarParameter_1 {class=MaterialExpressionScalarParameter, properties {parameterName=LOD}}
  MaterialExpressionScalarParameter_10 {class=MaterialExpressionScalarParameter, properties {defaultValue=1, group=Finish, parameterName=CCThickness_PaintColor}}
  MaterialExpressionScalarParameter_11 {class=MaterialExpressionScalarParameter, properties {defaultValue=0.1, group=Finish, parameterName=CCRoughness_TeamColor}}
  MaterialExpressionScalarParameter_12 {class=MaterialExpressionScalarParameter, properties {defaultValue=0.4, group=Finish, parameterName=CCRoughness_AccentColor}}
  MaterialExpressionScalarParameter_13 {class=MaterialExpressionScalarParameter, properties {defaultValue=0.4, group=Finish, parameterName=CCRoughness_PaintColor}}
  MaterialExpressionScalarParameter_14 {class=MaterialExpressionScalarParameter, properties {defaultValue=0.85, group=DetailNormal, parameterName=BottomNormalFlatness}}
  MaterialExpressionScalarParameter_15 {class=MaterialExpressionScalarParameter, properties {defaultValue=13, group=Finish, parameterName=BottomNormalTiling}}
  MaterialExpressionScalarParameter_16 {class=MaterialExpressionScalarParameter, properties {defaultValue=1, group=Color, parameterName=EmissiveBrightness, sortPriority=7}}
  MaterialExpressionScalarParameter_17 {class=MaterialExpressionScalarParameter, properties {parameterName=DiffuseEmissiveIntensity}}
  MaterialExpressionScalarParameter_18 {class=MaterialExpressionScalarParameter, properties {group=AlphaChannel, parameterName=RotatorSpeed_A}}
  MaterialExpressionScalarParameter_19 {class=MaterialExpressionScalarParameter, properties {defaultValue=0.5, group=BaseTextures, parameterName=NoiseIntensity_A}}
  MaterialExpressionScalarParameter_2 {class=MaterialExpressionScalarParameter, properties {defaultValue=0.7, group=Finish, parameterName=Metallic_TeamColor}}
  MaterialExpressionScalarParameter_20 {class=MaterialExpressionScalarParameter, properties {group=AlphaChannel, parameterName=DiffuseIntensity_A}}
  MaterialExpressionScalarParameter_21 {class=MaterialExpressionScalarParameter, properties {defaultValue=3, group=AlphaChannel, parameterName=EmissivePulseTiming_A}}
  MaterialExpressionScalarParameter_22 {class=MaterialExpressionScalarParameter, properties {defaultValue=0.25, group=AlphaChannel, parameterName=PulseInfluece_A}}
  MaterialExpressionScalarParameter_23 {class=MaterialExpressionScalarParameter, properties {defaultValue=0.2, group=AlphaChannel, parameterName=EmssiveIntensity_A}}
  MaterialExpressionScalarParameter_25 {class=MaterialExpressionScalarParameter, properties {defaultValue=0.1, group=AlphaChannel, parameterName=NoiseSubtractIntensity_A}}
  MaterialExpressionScalarParameter_26 {class=MaterialExpressionScalarParameter, properties {parameterName=LogoEmissive}}
  MaterialExpressionScalarParameter_27 {class=MaterialExpressionScalarParameter, properties {defaultValue=0.01, group=BlueChannel, parameterName=NoiseIntensity_B}}
  MaterialExpressionScalarParameter_28 {class=MaterialExpressionScalarParameter, properties {group=BlueChannel, parameterName=RotatorSpeed_B}}
  MaterialExpressionScalarParameter_29 {class=MaterialExpressionScalarParameter, properties {defaultValue=0.1, group=BlueChannel, parameterName=NoiseSubtractIntensity_B}}
  MaterialExpressionScalarParameter_3 {class=MaterialExpressionScalarParameter, properties {defaultValue=0.7, group=Finish, parameterName=Metallic_AccentColor}}
  MaterialExpressionScalarParameter_30 {class=MaterialExpressionScalarParameter, properties {defaultValue=-0.5, group=BlueChannel, parameterName=DiffuseIntensity_B}}
  MaterialExpressionScalarParameter_31 {class=MaterialExpressionScalarParameter, properties {defaultValue=-1, group=BlueChannel, parameterName=EmissivePulseTiming_B}}
  MaterialExpressionScalarParameter_32 {class=MaterialExpressionScalarParameter, properties {defaultValue=1, group=BlueChannel, parameterName=PulseInfluece_B}}
  MaterialExpressionScalarParameter_33 {class=MaterialExpressionScalarParameter, properties {defaultValue=0.5, parameterName=EmssiveIntensity_B}}
  MaterialExpressionScalarParameter_4 {class=MaterialExpressionScalarParameter, properties {defaultValue=0.7, group=Finish, parameterName=Metallic_PaintColor}}
  MaterialExpressionScalarParameter_41 {class=MaterialExpressionScalarParameter, properties {defaultValue=0.5, group=GreenChannel, parameterName=NoiseIntensity_G}}
  MaterialExpressionScalarParameter_42 {class=MaterialExpressionScalarParameter, properties {group=GreenChannel, parameterName=RotatorSpeed_G}}
  MaterialExpressionScalarParameter_43 {class=MaterialExpressionScalarParameter, properties {defaultValue=0.1, group=GreenChannel, parameterName=NoiseSubtractIntensity_G}}
  MaterialExpressionScalarParameter_44 {class=MaterialExpressionScalarParameter, properties {group=GreenChannel, parameterName=DiffuseIntensity_G}}
  MaterialExpressionScalarParameter_45 {class=MaterialExpressionScalarParameter, properties {defaultValue=3, group=GreenChannel, parameterName=EmissivePulseTiming_G}}
  MaterialExpressionScalarParameter_46 {class=MaterialExpressionScalarParameter, properties {defaultValue=0.25, group=GreenChannel, parameterName=PulseInfluece_G}}
  MaterialExpressionScalarParameter_47 {class=MaterialExpressionScalarParameter, properties {defaultValue=3, group=GreenChannel, parameterName=EmssiveIntensity_G}}
  MaterialExpressionScalarParameter_49 {class=MaterialExpressionScalarParameter, properties {defaultValue=0.5, group=RedChannel, parameterName=NoiseIntensity_R}}
  MaterialExpressionScalarParameter_5 {class=MaterialExpressionScalarParameter, properties {defaultValue=0.4, group=Finish, parameterName=Rough_TeamColor}}
  MaterialExpressionScalarParameter_50 {class=MaterialExpressionScalarParameter, properties {group=RedChannel, parameterName=RotatorSpeed_R}}
  MaterialExpressionScalarParameter_51 {class=MaterialExpressionScalarParameter, properties {defaultValue=0.1, group=RedChannel, parameterName=NoiseSubtractIntensity_R}}
  MaterialExpressionScalarParameter_52 {class=MaterialExpressionScalarParameter, properties {group=RedChannel, parameterName=DiffuseIntensity_R}}
  MaterialExpressionScalarParameter_53 {class=MaterialExpressionScalarParameter, properties {defaultValue=1, group=RedChannel, parameterName=EmissivePulseTiming_R}}
  MaterialExpressionScalarParameter_54 {class=MaterialExpressionScalarParameter, properties {defaultValue=1, group=RedChannel, parameterName=PulseInfluece_R}}
  MaterialExpressionScalarParameter_55 {class=MaterialExpressionScalarParameter, properties {defaultValue=3, group=RedChannel, parameterName=EmssiveIntensity_R}}
  MaterialExpressionScalarParameter_6 {class=MaterialExpressionScalarParameter, properties {defaultValue=0.5, group=Finish, parameterName=Rough_AccentColor}}
  MaterialExpressionScalarParameter_7 {class=MaterialExpressionScalarParameter, properties {defaultValue=0.5, group=Finish, parameterName=Rough_PaintColor}}
  MaterialExpressionScalarParameter_8 {class=MaterialExpressionScalarParameter, properties {defaultValue=1, group=Finish, parameterName=CCThickness_TeamColor}}
  MaterialExpressionScalarParameter_9 {class=MaterialExpressionScalarParameter, properties {defaultValue=1, group=Finish, parameterName=CCThickness_AccentColor}}
  MaterialExpressionSetMaterialAttributes_0 {
   class=MaterialExpressionSetMaterialAttributes
   inputs [
    MaterialAttributes {from="", output=""}
    BaseColor {from=MaterialExpressionNamedRerouteUsage_16, output=""}
    Metallic {from=MaterialExpressionNamedRerouteUsage_36, output=""}
    Roughness {from=MaterialExpressionNamedRerouteUsage_37, output=""}
    CustomData0 {from=MaterialExpressionNamedRerouteUsage_22, output=""}
    CustomData1 {from=MaterialExpressionNamedRerouteUsage_23, output=""}
    Normal {from=MaterialExpressionNamedRerouteUsage_3, output=""}
    ClearCoatBottomNormal {from=MaterialExpressionNamedRerouteUsage_32, output=""}
    EmissiveColor {from=MaterialExpressionNamedRerouteUsage_0, output=""}
   ]
   properties {attributeSetTypes [
     #1
     #2
     #3
     #4
     #5
     #6
     #7
     #8
    ]}
  }
  MaterialExpressionSine_0 {class=MaterialExpressionSine, inputs [None {from=MaterialExpressionMultiply_16}], properties {period=3}}
  MaterialExpressionSine_1 {class=MaterialExpressionSine, inputs [None {from=MaterialExpressionMultiply_26}], properties {period=3}}
  MaterialExpressionSine_3 {class=MaterialExpressionSine, inputs [None {from=MaterialExpressionMultiply_44}], properties {period=3}}
  MaterialExpressionSine_4 {class=MaterialExpressionSine, inputs [None {from=MaterialExpressionMultiply_53}], properties {period=3}}
  MaterialExpressionStaticComponentMaskParameter_0 {
   class=MaterialExpressionStaticComponentMaskParameter
   inputs [None {from=MaterialExpressionAppendVector_4, output=""}]
   properties {defaultB=true, group=BlueChannel, parameterName=NoiseChannel_A}
  }
  MaterialExpressionStaticComponentMaskParameter_1 {
   class=MaterialExpressionStaticComponentMaskParameter
   inputs [None {from=MaterialExpressionAppendVector_9, output=""}]
   properties {defaultA=true, group=BlueChannel, parameterName=NoiseChannel_B}
  }
  MaterialExpressionStaticComponentMaskParameter_3 {
   class=MaterialExpressionStaticComponentMaskParameter
   inputs [None {from=MaterialExpressionAppendVector_17, output=""}]
   properties {defaultR=true, group=BlueChannel, parameterName=NoiseChannel_G}
  }
  MaterialExpressionStaticComponentMaskParameter_4 {
   class=MaterialExpressionStaticComponentMaskParameter
   inputs [None {from=MaterialExpressionAppendVector_21, output=""}]
   properties {defaultR=true, group=BlueChannel, parameterName=NoiseChannel_R}
  }
  MaterialExpressionStaticSwitchParameter_0 {
   class=MaterialExpressionStaticSwitchParameter
   inputs [
    True {from=MaterialExpressionLinearInterpolate_10, output=""}
    False {from=MaterialExpressionReroute_27, output=BaseColor}
   ]
   properties {parameterName=UseBaseColorTexture}
  }
  MaterialExpressionStaticSwitchParameter_1 {
   class=MaterialExpressionStaticSwitchParameter
   inputs [
    True {from=MaterialExpressionSaturate_0, output=""}
    False {from=MaterialExpressionReroute_42, output=RGB}
   ]
   properties {
    desc="DO NOT USE FOR NEW ASSETS! This is intended for porting content from Rocket League only."
    group=Legacy_PORTINGONLY
    parameterName=SharpMasks
   }
  }
  MaterialExpressionStaticSwitchParameter_10 {
   class=MaterialExpressionStaticSwitchParameter
   inputs [
    True {from=MaterialExpressionRotator_0, output=""}
    False {from=MaterialExpressionStaticSwitchParameter_9, output=""}
   ]
   properties {group=AlphaChannel, parameterName=Rotator_A}
  }
  MaterialExpressionStaticSwitchParameter_11 {
   class=MaterialExpressionStaticSwitchParameter
   inputs [
    True {from=MaterialExpressionConstantBiasScale_0, output=""}
    False {from=MaterialExpressionStaticComponentMaskParameter_0, output=""}
   ]
   properties {defaultValue=true, group=AlphaChannel, parameterName=NoiseCBS_A}
  }
  MaterialExpressionStaticSwitchParameter_12 {
   class=MaterialExpressionStaticSwitchParameter
   inputs [
    True {from=MaterialExpressionMultiply_14, output=""}
    False {from=MaterialExpressionConstant_3, output=""}
   ]
   properties {defaultValue=true, group=AlphaChannel, parameterName=Noise_A}
  }
  MaterialExpressionStaticSwitchParameter_13 {
   class=MaterialExpressionStaticSwitchParameter
   inputs [
    True {from=MaterialExpressionVectorParameter_7, output=RGB}
    False {from=MaterialExpressionVectorParameter_6, output=RGB}
   ]
   properties {defaultValue=true, group=AlphaChannel, parameterName=UseCustomColor_A}
  }
  MaterialExpressionStaticSwitchParameter_14 {
   class=MaterialExpressionStaticSwitchParameter
   inputs [
    True {from=MaterialExpressionVectorParameter_9, output=RGB}
    False {from=MaterialExpressionStaticSwitchParameter_13, output=""}
   ]
   properties {group=AlphaChannel, parameterName=UseUniqueColor_A}
  }
  MaterialExpressionStaticSwitchParameter_15 {
   class=MaterialExpressionStaticSwitchParameter
   inputs [
    True {from=MaterialExpressionLinearInterpolate_19, output=""}
    False {from=MaterialExpressionStaticSwitchParameter_16, output=""}
   ]
   properties {defaultValue=true, group=AlphaChannel, parameterName=AlphaChannelBehindDiffuse}
  }
  MaterialExpressionStaticSwitchParameter_16 {
   class=MaterialExpressionStaticSwitchParameter
   inputs [
    True {from=MaterialExpressionClamp_2, output=""}
    False {from=MaterialExpressionTextureSample_2, output=A}
   ]
   properties {parameterName=NoiseSubtract_A}
  }
  MaterialExpressionStaticSwitchParameter_17 {
   class=MaterialExpressionStaticSwitchParameter
   inputs [
    True {from=MaterialExpressionMultiply_17, output=""}
    False {from=MaterialExpressionScalarParameter_23, output=""}
   ]
   properties {defaultValue=true, group=RedChannel, parameterName=EmissivePulse}
  }
  MaterialExpressionStaticSwitchParameter_18 {
   class=MaterialExpressionStaticSwitchParameter
   inputs [
    True {from=MaterialExpressionLinearInterpolate_21, output=""}
    False {from=MaterialExpressionConstant_7, output=""}
   ]
   properties {defaultValue=true, group=AlphaChannel, parameterName=Emissive_A}
  }
  MaterialExpressionStaticSwitchParameter_19 {
   class=MaterialExpressionStaticSwitchParameter
   inputs [
    True {from=MaterialExpressionOneMinus_0, output=""}
    False {from=MaterialExpressionStaticComponentMaskParameter_0, output=""}
   ]
   properties {parameterName=NoiseSubtractInvert_A}
  }
  MaterialExpressionStaticSwitchParameter_2 {
   class=MaterialExpressionStaticSwitchParameter
   inputs [
    True {from=MaterialExpressionMultiply_1, output=""}
    False {from=MaterialExpressionReroute_51, output=""}
   ]
   properties {
    desc="DO NOT USE FOR NEW ASSETS! This is intended for porting content from Rocket League only."
    group=Legacy_PORTINGONLY
    parameterName=SharpMasks
   }
  }
  MaterialExpressionStaticSwitchParameter_20 {
   class=MaterialExpressionStaticSwitchParameter
   inputs [
    True {from=MaterialExpressionConstantBiasScale_3, output=""}
    False {from=MaterialExpressionStaticComponentMaskParameter_1, output=""}
   ]
   properties {defaultValue=true, group=BlueChannel, parameterName=NoiseCBS_B}
  }
  MaterialExpressionStaticSwitchParameter_21 {
   class=MaterialExpressionStaticSwitchParameter
   inputs [
    True {from=MaterialExpressionMultiply_23, output=""}
    False {from=MaterialExpressionConstant_11, output=""}
   ]
   properties {defaultValue=true, group=BlueChannel, parameterName=Noise_B}
  }
  MaterialExpressionStaticSwitchParameter_22 {
   class=MaterialExpressionStaticSwitchParameter
   inputs [
    True {from=MaterialExpressionRotator_1, output=""}
    False {from=MaterialExpressionStaticSwitchParameter_23, output=""}
   ]
   properties {group=BlueChannel, parameterName=Rotator_B}
  }
  MaterialExpressionStaticSwitchParameter_23 {
   class=MaterialExpressionStaticSwitchParameter
   inputs [
    True {from=MaterialExpressionTextureCoordinate_11, output=""}
    False {from=MaterialExpressionTextureCoordinate_12, output=""}
   ]
   properties {defaultValue=true, group=BlueChannel, parameterName=UniversalUVs_B}
  }
  MaterialExpressionStaticSwitchParameter_24 {
   class=MaterialExpressionStaticSwitchParameter
   inputs [
    True {from=MaterialExpressionOneMinus_1, output=""}
    False {from=MaterialExpressionStaticComponentMaskParameter_1, output=""}
   ]
   properties {group=BlueChannel, parameterName=NoiseSubtractInvert_B}
  }
  MaterialExpressionStaticSwitchParameter_25 {
   class=MaterialExpressionStaticSwitchParameter
   inputs [
    True {from=MaterialExpressionVectorParameter_7, output=RGB}
    False {from=MaterialExpressionVectorParameter_6, output=RGB}
   ]
   properties {defaultValue=true, group=BlueChannel, parameterName=UseCustomColor_B}
  }
  MaterialExpressionStaticSwitchParameter_26 {
   class=MaterialExpressionStaticSwitchParameter
   inputs [
    True {from=MaterialExpressionVectorParameter_11, output=RGB}
    False {from=MaterialExpressionStaticSwitchParameter_25, output=""}
   ]
   properties {group=BlueChannel, parameterName=UseUniqueColor_B}
  }
  MaterialExpressionStaticSwitchParameter_27 {
   class=MaterialExpressionStaticSwitchParameter
   inputs [
    True {from=MaterialExpressionLinearInterpolate_25, output=""}
    False {from=MaterialExpressionStaticSwitchParameter_28, output=""}
   ]
   properties {group=BlueChannel, parameterName=BlueChannelBehindDiffuse}
  }
  MaterialExpressionStaticSwitchParameter_28 {
   class=MaterialExpressionStaticSwitchParameter
   inputs [
    True {from=MaterialExpressionClamp_4, output=""}
    False {from=MaterialExpressionTextureSample_3, output=B}
   ]
   properties {group=BlueChannel, parameterName=NoiseSubtract_B}
  }
  MaterialExpressionStaticSwitchParameter_29 {
   class=MaterialExpressionStaticSwitchParameter
   inputs [
    True {from=MaterialExpressionMultiply_27, output=""}
    False {from=MaterialExpressionScalarParameter_33, output=""}
   ]
   properties {defaultValue=true, group=BlueChannel, parameterName=EmissivePulse_B}
  }
  MaterialExpressionStaticSwitchParameter_3 {
   class=MaterialExpressionStaticSwitchParameter
   inputs [
    True {from=MaterialExpressionTextureCoordinate_4, output=""}
    False {from=MaterialExpressionTextureCoordinate_5, output=""}
   ]
   properties {group=DetailNormal, parameterName=UVChannel2}
  }
  MaterialExpressionStaticSwitchParameter_30 {
   class=MaterialExpressionStaticSwitchParameter
   inputs [
    True {from=MaterialExpressionLinearInterpolate_27, output=""}
    False {from=MaterialExpressionStaticSwitchParameter_18, output=""}
   ]
   properties {defaultValue=true, group=BlueChannel, parameterName=Emissive_B}
  }
  MaterialExpressionStaticSwitchParameter_31 {
   class=MaterialExpressionStaticSwitchParameter
   inputs [
    True {from=MaterialExpressionAppendVector_8, output=""}
    False {from=MaterialExpressionTextureCoordinate_10, output=""}
   ]
   properties {defaultValue=true, parameterName=LogoRotate90}
  }
  MaterialExpressionStaticSwitchParameter_33 {
   class=MaterialExpressionStaticSwitchParameter
   inputs [
    True {from=MaterialExpressionLinearInterpolate_29, output=""}
    False {from=MaterialExpressionStaticSwitchParameter_64, output=""}
   ]
   properties {defaultValue=true, parameterName=Logo}
  }
  MaterialExpressionStaticSwitchParameter_34 {
   class=MaterialExpressionStaticSwitchParameter
   inputs [
    True {from=MaterialExpressionLinearInterpolate_28, output=""}
    False {from=MaterialExpressionLinearInterpolate_39, output=""}
   ]
   properties {defaultValue=true, parameterName=Logo}
  }
  MaterialExpressionStaticSwitchParameter_4 {
   class=MaterialExpressionStaticSwitchParameter
   inputs [
    True {from=MaterialExpressionMultiply_7, output=""}
    False {from=MaterialExpressionConstant_1, output=""}
   ]
   properties {desc=Enables an Emissive Color parameter that is driven by a mask in the BodyTextures texture, group=Color, parameterName=UseEmissiveColor}
  }
  MaterialExpressionStaticSwitchParameter_42 {
   class=MaterialExpressionStaticSwitchParameter
   inputs [
    True {from=MaterialExpressionConstantBiasScale_7, output=""}
    False {from=MaterialExpressionStaticComponentMaskParameter_3, output=""}
   ]
   properties {group=GreenChannel, parameterName=NoiseCBS_G}
  }
  MaterialExpressionStaticSwitchParameter_43 {
   class=MaterialExpressionStaticSwitchParameter
   inputs [
    True {from=MaterialExpressionMultiply_41, output=""}
    False {from=MaterialExpressionConstant_22, output=""}
   ]
   properties {defaultValue=true, group=GreenChannel, parameterName=Noise_G}
  }
  MaterialExpressionStaticSwitchParameter_44 {
   class=MaterialExpressionStaticSwitchParameter
   inputs [
    True {from=MaterialExpressionRotator_3, output=""}
    False {from=MaterialExpressionStaticSwitchParameter_45, output=""}
   ]
   properties {group=GreenChannel, parameterName=Rotator_G}
  }
  MaterialExpressionStaticSwitchParameter_45 {
   class=MaterialExpressionStaticSwitchParameter
   inputs [
    True {from=MaterialExpressionTextureCoordinate_17, output=""}
    False {from=MaterialExpressionTextureCoordinate_18, output=""}
   ]
   properties {defaultValue=true, group=GreenChannel, parameterName=UniversalUVs_G}
  }
  MaterialExpressionStaticSwitchParameter_46 {
   class=MaterialExpressionStaticSwitchParameter
   inputs [
    True {from=MaterialExpressionOneMinus_3, output=""}
    False {from=MaterialExpressionStaticComponentMaskParameter_3, output=""}
   ]
   properties {group=GreenChannel, parameterName=NoiseSubtractInvert_G}
  }
  MaterialExpressionStaticSwitchParameter_47 {
   class=MaterialExpressionStaticSwitchParameter
   inputs [
    True {from=MaterialExpressionVectorParameter_7, output=RGB}
    False {from=MaterialExpressionVectorParameter_6, output=RGB}
   ]
   properties {defaultValue=true, group=GreenChannel, parameterName=UseCustomColor_G}
  }
  MaterialExpressionStaticSwitchParameter_48 {
   class=MaterialExpressionStaticSwitchParameter
   inputs [
    True {from=MaterialExpressionVectorParameter_15, output=RGB}
    False {from=MaterialExpressionStaticSwitchParameter_47, output=""}
   ]
   properties {group=GreenChannel, parameterName=UseUniqueColor_G}
  }
  MaterialExpressionStaticSwitchParameter_49 {
   class=MaterialExpressionStaticSwitchParameter
   inputs [
    True {from=MaterialExpressionLinearInterpolate_35, output=""}
    False {from=MaterialExpressionStaticSwitchParameter_53, output=""}
   ]
   properties {defaultValue=true, group=GreenChannel, parameterName=GreenChannelBehindDiffuse}
  }
  MaterialExpressionStaticSwitchParameter_5 {
   class=MaterialExpressionStaticSwitchParameter
   inputs [
    True {from=MaterialExpressionComponentMask_2, output=""}
    False {from=MaterialExpressionNamedRerouteUsage_28, output=""}
   ]
   properties {
    desc="When using, ensure PBR texture has \"Compress Without Alpha\" set to False. "
    group=Finish
    parameterName=UsePBR_BodyTextures
    sortPriority=-10
   }
  }
  MaterialExpressionStaticSwitchParameter_50 {
   class=MaterialExpressionStaticSwitchParameter
   inputs [
    True {from=MaterialExpressionClamp_8, output=""}
    False {from=MaterialExpressionTextureSample_5, output=G}
   ]
   properties {group=GreenChannel, parameterName=NoiseSubtract_G}
  }
  MaterialExpressionStaticSwitchParameter_51 {
   class=MaterialExpressionStaticSwitchParameter
   inputs [
    True {from=MaterialExpressionMultiply_45, output=""}
    False {from=MaterialExpressionScalarParameter_47, output=""}
   ]
   properties {defaultValue=true, group=GreenChannel, parameterName=EmissivePulse_G}
  }
  MaterialExpressionStaticSwitchParameter_52 {
   class=MaterialExpressionStaticSwitchParameter
   inputs [
    True {from=MaterialExpressionLinearInterpolate_37, output=""}
    False {from=MaterialExpressionConstant_26, output=""}
   ]
   properties {defaultValue=true, group=GreenChannel, parameterName=Emissive_G}
  }
  MaterialExpressionStaticSwitchParameter_53 {
   class=MaterialExpressionStaticSwitchParameter
   inputs [
    True {from=MaterialExpressionStaticSwitchParameter_50, output=""}
    False {from=MaterialExpressionTextureSample_5, output=G}
   ]
   properties {defaultValue=true, group=GreenChannel, parameterName=Noise_G}
  }
  MaterialExpressionStaticSwitchParameter_54 {
   class=MaterialExpressionStaticSwitchParameter
   inputs [
    True {from=MaterialExpressionConstantBiasScale_9, output=""}
    False {from=MaterialExpressionStaticComponentMaskParameter_4, output=""}
   ]
   properties {group=RedChannel, parameterName=NoiseCBS_R}
  }
  MaterialExpressionStaticSwitchParameter_55 {
   class=MaterialExpressionStaticSwitchParameter
   inputs [
    True {from=MaterialExpressionMultiply_50, output=""}
    False {from=MaterialExpressionConstant_28, output=""}
   ]
   properties {group=RedChannel, parameterName=Noise_R}
  }
  MaterialExpressionStaticSwitchParameter_56 {
   class=MaterialExpressionStaticSwitchParameter
   inputs [
    True {from=MaterialExpressionRotator_4, output=""}
    False {from=MaterialExpressionStaticSwitchParameter_57, output=""}
   ]
   properties {group=RedChannel, parameterName=Rotator_R}
  }
  MaterialExpressionStaticSwitchParameter_57 {
   class=MaterialExpressionStaticSwitchParameter
   inputs [
    True {from=MaterialExpressionTextureCoordinate_20, output=""}
    False {from=MaterialExpressionTextureCoordinate_21, output=""}
   ]
   properties {defaultValue=true, group=RedChannel, parameterName=UniversalUVs_R}
  }
  MaterialExpressionStaticSwitchParameter_58 {
   class=MaterialExpressionStaticSwitchParameter
   inputs [
    True {from=MaterialExpressionOneMinus_4, output=""}
    False {from=MaterialExpressionStaticComponentMaskParameter_4, output=""}
   ]
   properties {group=RedChannel, parameterName=NoiseSubtractInvert_R}
  }
  MaterialExpressionStaticSwitchParameter_59 {
   class=MaterialExpressionStaticSwitchParameter
   inputs [
    True {from=MaterialExpressionVectorParameter_7, output=RGB}
    False {from=MaterialExpressionVectorParameter_6, output=RGB}
   ]
   properties {defaultValue=true, group=RedChannel, parameterName=UseCustomColor_R}
  }
  MaterialExpressionStaticSwitchParameter_6 {
   class=MaterialExpressionStaticSwitchParameter
   inputs [
    True {from=MaterialExpressionComponentMask_3, output=""}
    False {from=MaterialExpressionNamedRerouteDeclaration_11, output=""}
   ]
   properties {
    desc="When using, ensure PBR texture has \"Compress Without Alpha\" set to False. "
    group=Finish
    parameterName=UsePBR_BodyTextures
    sortPriority=-10
   }
  }
  MaterialExpressionStaticSwitchParameter_60 {
   class=MaterialExpressionStaticSwitchParameter
   inputs [
    True {from=MaterialExpressionVectorParameter_17, output=RGB}
    False {from=MaterialExpressionStaticSwitchParameter_59, output=""}
   ]
   properties {group=RedChannel, parameterName=UseUniqueColor_R}
  }
  MaterialExpressionStaticSwitchParameter_61 {
   class=MaterialExpressionStaticSwitchParameter
   inputs [
    True {from=MaterialExpressionLinearInterpolate_40, output=""}
    False {from=MaterialExpressionStaticSwitchParameter_62, output=""}
   ]
   properties {defaultValue=true, group=RedChannel, parameterName=RedChannelBehindDiffuse}
  }
  MaterialExpressionStaticSwitchParameter_62 {
   class=MaterialExpressionStaticSwitchParameter
   inputs [
    True {from=MaterialExpressionClamp_10, output=""}
    False {from=MaterialExpressionTextureSample_6, output=R}
   ]
   properties {group=RedChannel, parameterName=NoiseSubtract_R}
  }
  MaterialExpressionStaticSwitchParameter_63 {
   class=MaterialExpressionStaticSwitchParameter
   inputs [
    True {from=MaterialExpressionMultiply_54, output=""}
    False {from=MaterialExpressionScalarParameter_55, output=""}
   ]
   properties {defaultValue=true, group=RedChannel, parameterName=EmissivePulse}
  }
  MaterialExpressionStaticSwitchParameter_64 {
   class=MaterialExpressionStaticSwitchParameter
   inputs [
    True {from=MaterialExpressionLinearInterpolate_42, output=""}
    False {from=MaterialExpressionConstant_32, output=""}
   ]
   properties {defaultValue=true, group=RedChannel, parameterName=Emissive}
  }
  MaterialExpressionStaticSwitchParameter_7 {
   class=MaterialExpressionStaticSwitchParameter
   inputs [
    True {from=MaterialExpressionTextureCoordinate_2, output=""}
    False {from=MaterialExpressionTextureCoordinate_3, output=""}
   ]
   properties {defaultValue=true, parameterName=DisseMap_UVSet2}
  }
  MaterialExpressionStaticSwitchParameter_8 {
   class=MaterialExpressionStaticSwitchParameter
   inputs [
    True {from=MaterialExpressionLinearInterpolate_9, output=""}
    False {from=MaterialExpressionTextureSampleParameter2D_9, output=RGB}
   ]
   properties {parameterName=DiffuseAlphaMap_UsePainColorAsBG}
  }
  MaterialExpressionStaticSwitchParameter_9 {
   class=MaterialExpressionStaticSwitchParameter
   inputs [
    True {from=MaterialExpressionTextureCoordinate_6, output=""}
    False {from=MaterialExpressionTextureCoordinate_7, output=""}
   ]
   properties {defaultValue=true, group=AlphaChannel, parameterName=UniversalUVs_A}
  }
  MaterialExpressionSubtract_0 {
   class=MaterialExpressionSubtract
   inputs [
    A {from=MaterialExpressionTextureSample_2, output=A}
    B {from=MaterialExpressionLinearInterpolate_22}
   ]
  }
  MaterialExpressionSubtract_1 {
   class=MaterialExpressionSubtract
   inputs [
    A {from=MaterialExpressionReroute_58, output=BodyTextures}
    B {from=""}
   ]
   properties {constB=0.098}
  }
  MaterialExpressionSubtract_2 {
   class=MaterialExpressionSubtract
   inputs [
    A {from=MaterialExpressionTextureSample_3, output=B}
    B {from=MaterialExpressionLinearInterpolate_23}
   ]
  }
  MaterialExpressionSubtract_4 {
   class=MaterialExpressionSubtract
   inputs [
    A {from=MaterialExpressionTextureSample_5, output=G}
    B {from=MaterialExpressionLinearInterpolate_33}
   ]
  }
  MaterialExpressionSubtract_5 {
   class=MaterialExpressionSubtract
   inputs [
    A {from=MaterialExpressionTextureSample_6, output=R}
    B {from=MaterialExpressionLinearInterpolate_38}
   ]
  }
  MaterialExpressionTextureCoordinate_0 {class=MaterialExpressionTextureCoordinate}
  MaterialExpressionTextureCoordinate_1 {class=MaterialExpressionTextureCoordinate, properties {coordinateIndex=1}}
  MaterialExpressionTextureCoordinate_10 {class=MaterialExpressionTextureCoordinate, properties {coordinateIndex=1}}
  MaterialExpressionTextureCoordinate_11 {class=MaterialExpressionTextureCoordinate, properties {coordinateIndex=1}}
  MaterialExpressionTextureCoordinate_12 {class=MaterialExpressionTextureCoordinate}
  MaterialExpressionTextureCoordinate_16 {class=MaterialExpressionTextureCoordinate, properties {coordinateIndex=1}}
  MaterialExpressionTextureCoordinate_17 {class=MaterialExpressionTextureCoordinate, properties {coordinateIndex=1}}
  MaterialExpressionTextureCoordinate_18 {class=MaterialExpressionTextureCoordinate}
  MaterialExpressionTextureCoordinate_19 {class=MaterialExpressionTextureCoordinate, properties {coordinateIndex=1}}
  MaterialExpressionTextureCoordinate_2 {class=MaterialExpressionTextureCoordinate, properties {coordinateIndex=1}}
  MaterialExpressionTextureCoordinate_20 {class=MaterialExpressionTextureCoordinate, properties {coordinateIndex=1}}
  MaterialExpressionTextureCoordinate_21 {class=MaterialExpressionTextureCoordinate}
  MaterialExpressionTextureCoordinate_3 {class=MaterialExpressionTextureCoordinate}
  MaterialExpressionTextureCoordinate_4 {class=MaterialExpressionTextureCoordinate, properties {coordinateIndex=1}}
  MaterialExpressionTextureCoordinate_5 {class=MaterialExpressionTextureCoordinate}
  MaterialExpressionTextureCoordinate_6 {class=MaterialExpressionTextureCoordinate, properties {coordinateIndex=1}}
  MaterialExpressionTextureCoordinate_7 {class=MaterialExpressionTextureCoordinate}
  MaterialExpressionTextureCoordinate_8 {class=MaterialExpressionTextureCoordinate, properties {coordinateIndex=1}}
  MaterialExpressionTextureCoordinate_9 {class=MaterialExpressionTextureCoordinate, properties {coordinateIndex=1}}
  MaterialExpressionTextureSampleParameter2D_0 {
   class=MaterialExpressionTextureSampleParameter2D
   inputs [
    UVs {from="", output=""}
    Apply View MipBias {from="", output=""}
   ]
   properties {
    group=Texture
    parameterName=BodyTextures
    samplerSource=SSM_Wrap_WorldGroupSettings
    samplerType=SAMPLERTYPE_LinearColor
    texture=/VehicleCosmetics/Textures/T_Black_Linear
   }
  }
  MaterialExpressionTextureSampleParameter2D_1 {
   class=MaterialExpressionTextureSampleParameter2D
   inputs [
    UVs {from="", output=""}
    Apply View MipBias {from="", output=""}
   ]
   properties {group=Texture, parameterName=SkinMask, samplerSource=SSM_Wrap_WorldGroupSettings, texture=/VehicleCosmetics/Textures/T_Black}
  }
  MaterialExpressionTextureSampleParameter2D_10 {
   class=MaterialExpressionTextureSampleParameter2D
   inputs [
    UVs {from=MaterialExpressionAdd_7, output=""}
    Apply View MipBias {from="", output=""}
   ]
   properties {parameterName=NoiseMap, texture=&/T_NPCity_Noise}
  }
  MaterialExpressionTextureSampleParameter2D_11 {
   class=MaterialExpressionTextureSampleParameter2D
   inputs [
    UVs {from=MaterialExpressionAdd_3, output=""}
    Apply View MipBias {from="", output=""}
   ]
   properties {
    parameterName=Logo+Alpha
    samplerSource=SSM_Clamp_WorldGroupSettings
    texture=&/T_NPCity_Knights_Logo
   }
  }
  MaterialExpressionTextureSampleParameter2D_12 {
   class=MaterialExpressionTextureSampleParameter2D
   inputs [
    UVs {from=MaterialExpressionAdd_11, output=""}
    Apply View MipBias {from="", output=""}
   ]
   properties {parameterName=NoiseMap, texture=&T_NPCity_Noise}
  }
  MaterialExpressionTextureSampleParameter2D_13 {
   class=MaterialExpressionTextureSampleParameter2D
   inputs [
    UVs {from=MaterialExpressionAdd_13, output=""}
    Apply View MipBias {from="", output=""}
   ]
   properties {parameterName=NoiseMap, texture=&T_NPCity_Noise}
  }
  MaterialExpressionTextureSampleParameter2D_2 {
   class=MaterialExpressionTextureSampleParameter2D
   inputs [
    UVs {from=MaterialExpressionReroute_4, output=UV}
    Apply View MipBias {from="", output=""}
   ]
   properties {group=Texture, parameterName=SkinBaseColor, samplerSource=SSM_Wrap_WorldGroupSettings, texture=&T_Black}
  }
  MaterialExpressionTextureSampleParameter2D_3 {
   class=MaterialExpressionTextureSampleParameter2D
   inputs [
    UVs {from=MaterialExpressionReroute_12, output=UV Detail}
    Apply View MipBias {from="", output=""}
   ]
   properties {group=Finish, parameterName=DetailBase, samplerSource=SSM_Wrap_WorldGroupSettings, texture=/VehicleCosmetics/Textures/T_White}
  }
  MaterialExpressionTextureSampleParameter2D_4 {
   class=MaterialExpressionTextureSampleParameter2D
   inputs [
    UVs {from=MaterialExpressionReroute_12, output=UV Detail}
    Apply View MipBias {from="", output=""}
   ]
   properties {group=Finish, parameterName=DetailAccent, samplerSource=SSM_Wrap_WorldGroupSettings, texture=&T_White}
  }
  MaterialExpressionTextureSampleParameter2D_5 {
   class=MaterialExpressionTextureSampleParameter2D
   inputs [
    UVs {from="", output=""}
    Apply View MipBias {from="", output=""}
   ]
   properties {group=Texture, parameterName=BodyNormalLOD, samplerType=SAMPLERTYPE_Normal, texture=/VehicleCosmetics/Textures/T_BlankNormal}
  }
  MaterialExpressionTextureSampleParameter2D_6 {
   class=MaterialExpressionTextureSampleParameter2D
   inputs [
    UVs {from=MaterialExpressionReroute_41, output=UV BottomNormal}
    Apply View MipBias {from="", output=""}
   ]
   properties {
    group=Finish
    parameterName=BottomNormalBase
    samplerType=SAMPLERTYPE_Normal
    texture=/VehicleCosmetics/SharedTextures/T_Wrap_metallicFlakes_Small
   }
  }
  MaterialExpressionTextureSampleParameter2D_7 {
   class=MaterialExpressionTextureSampleParameter2D
   inputs [
    UVs {from=MaterialExpressionReroute_41, output=UV BottomNormal}
    Apply View MipBias {from="", output=""}
   ]
   properties {group=Finish, parameterName=BottomNormalAccent, samplerType=SAMPLERTYPE_Normal, texture=/VehicleCosmetics/Textures/T_BlankNormal_B}
  }
  MaterialExpressionTextureSampleParameter2D_8 {
   class=MaterialExpressionTextureSampleParameter2D
   inputs [
    UVs {from=MaterialExpressionAdd_5, output=""}
    Apply View MipBias {from="", output=""}
   ]
   properties {parameterName=NoiseMap, texture=&T_NPCity_Noise}
  }
  MaterialExpressionTextureSampleParameter2D_9 {
   class=MaterialExpressionTextureSampleParameter2D
   inputs [
    UVs {from=MaterialExpressionStaticSwitchParameter_7, output=""}
    Apply View MipBias {from="", output=""}
   ]
   properties {
    automaticViewMipBias=false
    parameterName=DiffuseAlphaMap
    texture=&/T_NPCity_Knights_Diffuse_MC1
   }
  }
  MaterialExpressionTextureSample_2 {
   class=MaterialExpressionTextureSample
   inputs [
    UVs {from=MaterialExpressionAdd_6}
    Tex {from=""}
    Apply View MipBias {from=""}
   ]
   properties {texture=&/T_NPCity_Knights_Masks_MC1}
  }
  MaterialExpressionTextureSample_3 {
   class=MaterialExpressionTextureSample
   inputs [
    UVs {from=MaterialExpressionAdd_8}
    Tex {from=""}
    Apply View MipBias {from=""}
   ]
   properties {texture=&T_NPCity_Knights_Masks_MC1}
  }
  MaterialExpressionTextureSample_5 {
   class=MaterialExpressionTextureSample
   inputs [
    UVs {from=MaterialExpressionAdd_12}
    Tex {from=""}
    Apply View MipBias {from=""}
   ]
   properties {texture=&T_NPCity_Knights_Masks_MC1}
  }
  MaterialExpressionTextureSample_6 {
   class=MaterialExpressionTextureSample
   inputs [
    UVs {from=MaterialExpressionAdd_14}
    Tex {from=""}
    Apply View MipBias {from=""}
   ]
   properties {texture=&T_NPCity_Knights_Masks_MC1}
  }
  MaterialExpressionTime_0 {class=MaterialExpressionTime}
  MaterialExpressionTime_1 {class=MaterialExpressionTime}
  MaterialExpressionTime_10 {class=MaterialExpressionTime}
  MaterialExpressionTime_11 {class=MaterialExpressionTime}
  MaterialExpressionTime_12 {class=MaterialExpressionTime}
  MaterialExpressionTime_13 {class=MaterialExpressionTime}
  MaterialExpressionTime_14 {class=MaterialExpressionTime}
  MaterialExpressionTime_15 {class=MaterialExpressionTime}
  MaterialExpressionTime_2 {class=MaterialExpressionTime}
  MaterialExpressionTime_3 {class=MaterialExpressionTime}
  MaterialExpressionTime_5 {class=MaterialExpressionTime}
  MaterialExpressionTime_6 {class=MaterialExpressionTime}
  MaterialExpressionVectorParameter_0 {class=MaterialExpressionVectorParameter, properties {group=Color, parameterName=BaseColor, sortPriority=1, defaultValue=<rgba 0 0.2 1 1>}}
  MaterialExpressionVectorParameter_1 {class=MaterialExpressionVectorParameter, properties {group=Color, parameterName=AccentColor, sortPriority=2, defaultValue=<rgba 1 1 1 1>}}
  MaterialExpressionVectorParameter_10 {class=MaterialExpressionVectorParameter, properties {group=BlueChannel, parameterName=NoiseTransforms_B, defaultValue=<rgba 10 10 0 0.05>}}
  MaterialExpressionVectorParameter_11 {class=MaterialExpressionVectorParameter, properties {group=BlueChannel, parameterName=Color_B, defaultValue=<rgba 0 0 1 1>}}
  MaterialExpressionVectorParameter_12 {class=MaterialExpressionVectorParameter, properties {parameterName=LogoTransform, defaultValue=<rgba 13 13 -9.03 -7.5>}}
  MaterialExpressionVectorParameter_14 {class=MaterialExpressionVectorParameter, properties {group=GreenChannel, parameterName=NoiseTransforms_G, defaultValue=<rgba 3 2 0.1 0.5>}}
  MaterialExpressionVectorParameter_15 {class=MaterialExpressionVectorParameter, properties {group=GreenChannel, parameterName=Color_G, defaultValue=<rgba 0 1 0 0.1>}}
  MaterialExpressionVectorParameter_16 {class=MaterialExpressionVectorParameter, properties {group=RedChannel, parameterName=NoiseTransforms_R, defaultValue=<rgba 1.5 2 0.1 0.5>}}
  MaterialExpressionVectorParameter_17 {class=MaterialExpressionVectorParameter, properties {parameterName=Color_R, defaultValue=<rgba 1 0 0 1>}}
  MaterialExpressionVectorParameter_2 {
   class=MaterialExpressionVectorParameter
   properties {group=Color, parameterName=TrimColor, sortPriority=4, defaultValue=<rgba 0.03 0.03 0.03 0>}
  }
  MaterialExpressionVectorParameter_3 {class=MaterialExpressionVectorParameter, properties {group=Colors, parameterName=PaintColor, defaultValue=<rgba 0.0137 0.0137 0.0137 1>}}
  MaterialExpressionVectorParameter_4 {class=MaterialExpressionVectorParameter, properties {group=Colors, parameterName=PaintColor, defaultValue=<rgba 0.0137 0.0137 0.0137 1>}}
  MaterialExpressionVectorParameter_5 {class=MaterialExpressionVectorParameter, properties {group=Color, parameterName=EmissiveColor, sortPriority=5, defaultValue=<rgba 0 1 1 1>}}
  MaterialExpressionVectorParameter_6 {class=MaterialExpressionVectorParameter, properties {group=Colors, parameterName=TeamColor, defaultValue=<rgba 0.5029 0.0437 0.5029 1>}}
  MaterialExpressionVectorParameter_7 {class=MaterialExpressionVectorParameter, properties {group=Colors, parameterName=CustomColor, defaultValue=<rgba 1 0 0.1301 1>}}
  MaterialExpressionVectorParameter_8 {class=MaterialExpressionVectorParameter, properties {group=AlphaChannel, parameterName=NoiseTransforms_A, defaultValue=<rgba 1 0.5 0 0.05>}}
  MaterialExpressionVectorParameter_9 {class=MaterialExpressionVectorParameter, properties {group=AlphaChannel, parameterName=Color_A, defaultValue=<rgba 1 1 1 1>}}
 }}
